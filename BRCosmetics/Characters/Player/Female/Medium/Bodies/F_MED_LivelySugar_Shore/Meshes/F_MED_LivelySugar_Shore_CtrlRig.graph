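# Control Rig for the F_MED_LivelySugar_Shore player body that corrects deformation and secondary costume bones.
# After an optional debug control pass, FortDeformRig runs, then weighted parent constraints attach belts, holster, heel bullets, coat and necklace to body bones.
# Heel, belt, accessory and coat passes layer corrective offsets driven by remapped joint rotations, spherical pose readers and knee-slide, falling and skydiving states.
# Construction adds up-vector nulls for the heel aim constraints.
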ControlRigBlueprint /BRCosmetics/Characters/Player/Female/Medium/Bodies/F_MED_LivelySugar_Shore/Meshes/F_MED_LivelySugar_Shore_CtrlRig
graphs [
 New Function_ContainedGraph {nodes [
   RigVMFunctionEntryNode()
   > &RigUnit_ParentConstraint(
    AdvancedSettings=(RotationOrderForFilter=XZY)
    Child=(Type=Bone,Name="base_handkerchief")
    Parents=((Item=(Type=Bone,Name="thigh_twist_01_r"),Weight=0.0582),(Item=(Type=Bone,Name="pelvis"),Weight=0.9418))
    Weight=1
   )
   > &RigUnit_ParentConstraint(
    AdvancedSettings=(RotationOrderForFilter=XZY)
    Child=(Type=Bone,Name="base_beltBullets")
    Parents=((Item=(Type=Bone,Name="thigh_twist_01_r"),Weight=0.0586),(Item=(Type=Bone,Name="pelvis"),Weight=0.9414))
    Weight=1
   )
   > &RigUnit_ParentConstraint(
    AdvancedSettings=(RotationOrderForFilter=XZY)
    Child=(Type=Bone,Name="base_belt_r")
    Parents="((Item=(Type=Bone,Name=\"spine_01\"),Weight=0.0002),(Item=(Type=Bone,Name=\"spine_02\"),Weight=0.0003),(Item=(Type=Bone,Name=\"thigh_twist_01_l\"),Weight=0.0005),(Item=(Type=Bone,Name=\"thigh_r\"),Weight=0.0013),(Item=(Type=Bone,Name=\"thigh_twist_01_r\"),Weight=0.1 ...(truncated)"
    Weight=1
   )
   > &RigUnit_ParentConstraint(
    AdvancedSettings=(RotationOrderForFilter=XZY)
    Child=(Type=Bone,Name="base_belt_l")
    Parents=((Item=(Type=Bone,Name="thigh_twist_01_r"),Weight=0.0007),(Item=(Type=Bone,Name="thigh_l"),Weight=0.0258),(Item=(Type=Bone,Name="thigh_twist_01_l"),Weight=0.4704),(Item=(Type=Bone,Name="pelvis"),Weight=0.5031))
    Weight=1
   )
   > &RigUnit_ParentConstraint(
    AdvancedSettings=(RotationOrderForFilter=XZY)
    Child=(Type=Bone,Name="base_holster")
    Parents=((Item=(Type=Bone,Name="thigh_l"),Weight=0.0944),(Item=(Type=Bone,Name="thigh_twist_01_l"),Weight=0.4527),(Item=(Type=Bone,Name="pelvis"),Weight=0.4529))
    Weight=1
   )
   > &RigUnit_ParentConstraint(
    AdvancedSettings=(RotationOrderForFilter=XZY)
    Child=(Type=Bone,Name="base_heelBullet_r")
    Parents=((Item=(Type=Bone,Name="foot_r"),Weight=0.045),(Item=(Type=Bone,Name="calf_r"),Weight=0.0811),(Item=(Type=Bone,Name="calf_twist_01_r"),Weight=0.8738))
    Weight=1
   )
   > &RigUnit_ParentConstraint(
    AdvancedSettings=(RotationOrderForFilter=XZY)
    Child=(Type=Bone,Name="base_heelBullet_l")
    Parents=((Item=(Type=Bone,Name="foot_l"),Weight=0.045),(Item=(Type=Bone,Name="calf_l"),Weight=0.0811),(Item=(Type=Bone,Name="calf_twist_01_l"),Weight=0.8738))
    Weight=1
   )
   > &RigUnit_ParentConstraint(
    AdvancedSettings=(RotationOrderForFilter=XZY)
    Child=(Type=Bone,Name="base_coatVolume01_r")
    Filter=(TranslationFilter=(bX=True,bY=True,bZ=True),RotationFilter=(bX=False,bY=False,bZ=False),ScaleFilter=(bX=True,bY=True,bZ=True))
    Parents="((Item=(Type=Bone,Name=\"spine_02\"),Weight=0.0009),(Item=(Type=Bone,Name=\"deform_upperarm_up_r\"),Weight=0.0147),(Item=(Type=Bone,Name=\"deform_upperarm_fr_r\"),Weight=0.0928),(Item=(Type=Bone,Name=\"spine_05\"),Weight=0.1329),(Item=(Type=Bone,Name=\"clavicle_r\") ...(truncated)"
    Weight=1
   )
   > &RigUnit_ParentConstraint(
    AdvancedSettings=(RotationOrderForFilter=XZY)
    Child=(Type=Bone,Name="base_coatVolume01_l")
    Parents="((Item=(Type=Bone,Name=\"spine_02\"),Weight=0.001),(Item=(Type=Bone,Name=\"deform_upperarm_up_l\"),Weight=0.0154),(Item=(Type=Bone,Name=\"deform_upperarm_fr_l\"),Weight=0.1059),(Item=(Type=Bone,Name=\"spine_05\"),Weight=0.122),(Item=(Type=Bone,Name=\"clavicle_l\"),W ...(truncated)"
    Weight=1
   )
   > &RigUnit_ParentConstraint(
    AdvancedSettings=(RotationOrderForFilter=XZY)
    Child=(Type=Bone,Name="base_coatArmpit_l")
    Parents="((Item=(Type=Bone,Name=\"spine_03\"),Weight=0.0114),(Item=(Type=Bone,Name=\"spine_05\"),Weight=0.0156),(Item=(Type=Bone,Name=\"clavicle_l\"),Weight=0.0182),(Item=(Type=Bone,Name=\"spine_02\"),Weight=0.0185),(Item=(Type=Bone,Name=\"upperarm_l\"),Weight=0.0549),(Item= ...(truncated)"
    Weight=1
   )
   > &RigUnit_ParentConstraint(
    AdvancedSettings=(RotationOrderForFilter=XZY)
    Child=(Type=Bone,Name="base_coatArmpit_r")
    Parents="((Item=(Type=Bone,Name=\"spine_03\"),Weight=0.006),(Item=(Type=Bone,Name=\"spine_02\"),Weight=0.0148),(Item=(Type=Bone,Name=\"spine_05\"),Weight=0.0203),(Item=(Type=Bone,Name=\"clavicle_r\"),Weight=0.0241),(Item=(Type=Bone,Name=\"upperarm_r\"),Weight=0.0516),(Item=( ...(truncated)"
    Weight=1
   )
   > &RigUnit_ParentConstraint(
    AdvancedSettings=(RotationOrderForFilter=XZY)
    Child=(Type=Bone,Name="base_necklace")
    Parents="((Item=(Type=Bone,Name=\"clavicle_l\"),Weight=0.0049),(Item=(Type=Bone,Name=\"clavicle_r\"),Weight=0.0049),(Item=(Type=Bone,Name=\"spine_03\"),Weight=0.0061),(Item=(Type=Bone,Name=\"neck_01\"),Weight=0.0065),(Item=(Type=Bone,Name=\"spine_04\"),Weight=0.0678),(Item=( ...(truncated)"
    Weight=1
   )
   > RigVMFunctionReturnNode()
  ]}
 New Function_ContainedGraph {nodes [
   RigVMFunctionEntryNode()
   > RigVMFunction_Sequence = RigVMAggregateNode()
   Delta Rotation X_2 = RigVMFunctionReferenceNode(Child=heelFoot_l, Parent=heelFoot_l_null, Source Maximum=-205, Source Minimum=-10, ExecuteContext=@RigVMFunction_Sequence.B)
   > &RigUnit_RotationConstraintLocalSpaceOffset(
    AdvancedSettings=(InterpolationType=Average,RotationOrderForFilter=XZY)
    Child=(Type=Bone,Name="base_heelBullet_l")
    Filter=(bX=false,bY=false,bZ=True)
    Parents=((Item=(Type=Bone,Name="foot_l"),Weight=1))
    Weight=@"Delta Rotation X_2".Delta Rotation
   )
   > Delta Rotation X_3 = RigVMFunctionReferenceNode(Child=heelFoot_r, Parent=heelFoot_r_null, Source Maximum=-205, Source Minimum=-10)
   > &RigUnit_RotationConstraintLocalSpaceOffset(
    AdvancedSettings=(InterpolationType=Average,RotationOrderForFilter=XZY)
    Child=(Type=Bone,Name="base_heelBullet_r")
    Filter=(bX=false,bY=false,bZ=True)
    Parents=((Item=(Type=Bone,Name="foot_r"),Weight=1))
    Weight=@"Delta Rotation X_3".Delta Rotation
   )
   RigVMFunctionReturnNode(ExecuteContext=@RigVMFunction_Sequence.C)
   /Script/ControlRig.RigUnit_PositionConstraintLocalSpaceOffset(
    Child=(Type=Bone,Name="heelFoot_l")
    ExecutePin=()
    Filter=(bX=True,bY=True,bZ=True)
    Parents=((Item=(Type=Bone,Name="foot_l"),Weight=1))
    Weight=1
    bMaintainOffset=True
   ) {is=struct RigVMUnitNode}
   > &RigUnit_AimConstraintLocalSpaceOffset(
    AimAxis=(X=0,Y=1,Z=0)
    Child=(Type=Bone,Name="heelFoot_l")
    Filter=(bX=True,bY=true,bZ=true)
    Parents=((Item=(Type=Bone,Name="ball_l"),Weight=1))
    UpAxis=(X=0,Y=0,Z=-1)
    Weight=1
    WorldUp=(Target=(X=0,Y=0,Z=1),Kind=Location,Space=(Type=Null,Name="metaTarsoRet_l_upV_null"))
    bMaintainOffset=true
   )
   > &RigUnit_RotationConstraintLocalSpaceOffset(
    AdvancedSettings=(InterpolationType=Average,RotationOrderForFilter=XZY)
    Child=(Type=Bone,Name="heelBall_l")
    Filter=(bX=True,bY=True,bZ=True)
    Parents=((Item=(Type=Bone,Name="ball_l"),Weight=1))
    Weight=1
   )
   > &RigUnit_PositionConstraintLocalSpaceOffset(
    Child=(Type=Bone,Name="heelFoot_r")
    Filter=(bX=True,bY=True,bZ=True)
    Parents=((Item=(Type=Bone,Name="foot_r"),Weight=1))
    Weight=1
    bMaintainOffset=True
   )
   > &RigUnit_AimConstraintLocalSpaceOffset(
    AimAxis=(X=0,Y=-1,Z=0)
    Child=(Type=Bone,Name="heelFoot_r")
    Filter=(bX=True,bY=true,bZ=true)
    Parents=((Item=(Type=Bone,Name="ball_r"),Weight=1))
    Weight=1
    WorldUp=(Target=(X=0,Y=0,Z=1),Kind=Location,Space=(Type=Null,Name="metaTarsoRet_r_upV_null"))
    bMaintainOffset=true
   )
   > &RigUnit_RotationConstraintLocalSpaceOffset(
    AdvancedSettings=(InterpolationType=Average,RotationOrderForFilter=XZY)
    Child=(Type=Bone,Name="heelBall_r")
    Filter=(bX=True,bY=True,bZ=True)
    Parents=((Item=(Type=Bone,Name="ball_r"),Weight=1))
    Weight=1
   )
   RigVMFunction_MathBoolToFloat = &RigVMFunction_MathBoolToFloat(Value=RigVMVariableNode() {variable=isKneeSliding_CR}.Value)
   AlphaInterp = &RigVMFunction_AlphaInterp(InterpSpeedDecreasing=8, InterpSpeedIncreasing=8, Scale=1, bClampResult=False, Value=@RigVMFunction_MathBoolToFloat.Result)
   > &RigUnit_ModifyTransforms(
    ItemToModify="((Item=(Type=Bone,Name=\"heelBall_l\"),Transform=(Rotation=(X=0,Y=0,Z=-0.1305,W=0.9914),Translation=(X=0,Y=0,Z=0),Scale3D=(X=1,Y=1,Z=1))),(Item=(Type=Bone,Name=\"heelFoot_r\"),Transform=(Rotation=(X=0,Y=0,Z=-0.1305,W=0.9914),Translation=(X=0,Y=0,Z=0),Scale3D=( ...(truncated)"
    WeightMinimum=0
    Weight=@AlphaInterp.Result
   )
  ]}
 New Function_ContainedGraph {nodes [
   RigVMFunctionEntryNode()
   > RigVMFunction_Sequence = RigVMAggregateNode()
   Delta Rotation Z_4 = RigVMFunctionReferenceNode(Child=thigh_l, Parent=thigh_l_null, Source Maximum=-100, Source Minimum=-5, ExecuteContext=@RigVMFunction_Sequence.A)
   AlphaInterp = &RigVMFunction_AlphaInterp(InterpSpeedDecreasing=8, InterpSpeedIncreasing=8, Scale=1, bClampResult=False, Value=RigVMVariableNode() {variable=isFalling_CR}.Value)
   Remap = &RigVMFunction_MathDoubleRemap(SourceMaximum=1, SourceMinimum=0.3, TargetMaximum=1, TargetMinimum=0.6, Value=@AlphaInterp.Result)
   Multiply = &RigVMFunction_MathDoubleMul(A=@"Delta Rotation Z_4".Delta Rotation, B=@Remap.Result)
   > &RigUnit_ModifyTransforms(
    ItemToModify="((Item=(Type=Bone,Name=\"dyn_belt01\"),Transform=(Rotation=(X=-0.0009,Y=0.0203,Z=0.0433,W=0.9989),Translation=(X=0.0131,Y=0.0011,Z=-0.0005),Scale3D=(X=1,Y=1,Z=1))),(Item=(Type=Bone,Name=\"base_belt_l\"),Transform=(Rotation=(X=-0.1566,Y=0.2984,Z=-0.4375,W=-0.83 ...(truncated)"
    WeightMinimum=0
    Weight=@Multiply.Result
   )
   > Delta Rotation Z_5 = RigVMFunctionReferenceNode(Child=thigh_r, Parent=thigh_r_null, Source Maximum=-100, Source Minimum=-35)
   > &RigUnit_ModifyTransforms(
    ItemToModify="((Item=(Type=Bone,Name=\"beltRotation_r\"),Transform=(Rotation=(X=0.02,Y=0.0687,Z=0.0254,W=-0.9971),Translation=(X=-1.3546,Y=-1.8609,Z=-0.2416),Scale3D=(X=1,Y=1,Z=1))),(Item=(Type=Bone,Name=\"dyn_belt_l\"),Transform=(Rotation=(X=0.0006,Y=-0.2648,Z=0.0004,W=0.9 ...(truncated)"
    WeightMinimum=0
    Weight=@"Delta Rotation Z_5".Delta Rotation
   )
   > Delta Rotation Z_6 = RigVMFunctionReferenceNode(Child=thigh_r, Parent=thigh_r_null, Source Maximum=-70, Source Minimum=0)
   > &RigUnit_ModifyTransforms(
    ItemToModify=((Item=(Type=Bone,Name="base_handkerchief"),Transform=(Rotation=(X=-0.1679,Y=0.0474,Z=-0.1303,W=0.976),Translation=(X=0.808,Y=0.3499,Z=0.0634),Scale3D=(X=1,Y=1,Z=1))))
    WeightMinimum=0
    Weight=@"Delta Rotation Z_6".Delta Rotation
   )
   > Delta Rotation Y_3 = RigVMFunctionReferenceNode(Child=thigh_r, Parent=thigh_r_null, Source Maximum=-90, Source Minimum=0)
   > &RigUnit_ModifyTransforms(
    ItemToModify="((Item=(Type=Bone,Name=\"dyn_handkerchief\"),Transform=(Rotation=(X=0.1508,Y=-0.1085,Z=0.6246,W=-0.7585),Translation=(X=0,Y=0,Z=0),Scale3D=(X=1,Y=1,Z=1))),(Item=(Type=Bone,Name=\"dyn_beltBullets\"),Transform=(Rotation=(X=0.2902,Y=-0.0785,Z=-0.2014,W=0.9322),Tr ...(truncated)"
    WeightMinimum=0
    Weight=@"Delta Rotation Y_3".Delta Rotation
   )
   > Delta Rotation Y_4 = RigVMFunctionReferenceNode(Child=thigh_l, Parent=thigh_l_null, Source Maximum=-80, Source Minimum=0)
   > &RigUnit_ModifyTransforms(
    ItemToModify="((Item=(Type=Bone,Name=\"dyn_belt01\"),Transform=(Rotation=(X=0,Y=0.0874,Z=0,W=0.9962),Translation=(X=0,Y=0,Z=0),Scale3D=(X=1,Y=1,Z=1))),(Item=(Type=Bone,Name=\"dyn_belt_l\"),Transform=(Rotation=(X=0,Y=0,Z=0,W=1),Translation=(X=0,Y=0,Z=0),Scale3D=(X=1,Y=1,Z=1) ...(truncated)"
    WeightMinimum=0
    Weight=@"Delta Rotation Y_4".Delta Rotation
   )
   > Delta Rotation Z_7 = RigVMFunctionReferenceNode(Child=thigh_l, Parent=thigh_l_null, Source Maximum=60, Source Minimum=0)
   > &RigUnit_ModifyTransforms(
    ItemToModify=((Item=(Type=Bone,Name="base_holster"),Transform=(Rotation=(X=-0.105,Y=0.0393,Z=0.2471,W=0.9625),Translation=(X=-0.3792,Y=0.5092,Z=0.8291),Scale3D=(X=1,Y=1,Z=1))))
    WeightMinimum=0
    Weight=@"Delta Rotation Z_7".Delta Rotation
   )
   > Delta Rotation Z_1 = RigVMFunctionReferenceNode(Child=thigh_l, Parent=thigh_l_null, Source Maximum=20, Source Minimum=-40)
   > n1 = &RigUnit_ModifyTransforms(
    ItemToModify=((Item=(Type=Bone,Name="base_holster"),Transform=(Rotation=(X=0,Y=-0.0872,Z=0,W=0.9962),Translation=(X=0,Y=0,Z=-2),Scale3D=(X=1,Y=1,Z=1))))
    WeightMinimum=0
    Weight=&RigVMFunction_MathDoubleMul(A=@"Delta Rotation Z_1".Delta Rotation, B=@AlphaInterp.Result).Result
   )
   RigVMFunction_ControlFlowBranch = &RigVMFunction_ControlFlowBranch(ExecuteContext=@n1.ExecutePin, Condition=RigVMVariableNode() {variable=isSkydiving_CR}.Value)
   &RigUnit_ModifyTransforms(
    ItemToModify=((Item=(Type=Bone,Name="base_handkerchiefBottom"),Transform=(Rotation=(X=-0.633,Y=0.1116,Z=-0.133,W=0.7544),Translation=(X=0,Y=0,Z=0),Scale3D=(X=1,Y=1,Z=1))))
    Weight=1
    WeightMinimum=0
    ExecutePin=@RigVMFunction_ControlFlowBranch.True
   )
   RigVMFunctionReturnNode(ExecuteContext=@RigVMFunction_Sequence.C)
  ]}
 New Function_ContainedGraph {nodes [
   n1 = RigVMFunctionEntryNode()
   RigVMFunction_ControlFlowBranch = &RigVMFunction_ControlFlowBranch(Condition=@n1.Debug Controls, ExecuteContext=@n1.ExecuteContext)
   RigVMFunctionReturnNode(ExecuteContext=@RigVMFunction_ControlFlowBranch.Completed)
   RigUnit_GetTransform = &RigUnit_GetTransform(Item=(Type=Control,Name="thigh_r_ctrl"), bInitial=False)
   &RigUnit_SetTransform(
    Item=(Type=Bone,Name="thigh_r")
    Weight=1
    bInitial=False
    Value=@RigUnit_GetTransform.Transform
    ExecutePin=@RigVMFunction_ControlFlowBranch.True
   )
   > &RigUnit_SetTransform(
    Item=(Type=Bone,Name="thigh_l")
    Weight=1
    bInitial=False
    Value=&RigUnit_GetTransform(Item=(Type=Control,Name="thigh_l_ctrl"), bInitial=False).Transform
   )
   > &RigUnit_SetTransform(
    Item=(Type=Bone,Name="calf_l")
    Weight=1
    bInitial=False
    Value=&RigUnit_GetTransform(Item=(Type=Control,Name="calf_l_ctrl"), bInitial=False).Transform
   )
   > &RigUnit_SetTransform(
    Item=(Type=Bone,Name="calf_r")
    Weight=1
    bInitial=False
    Value=&RigUnit_GetTransform(Item=(Type=Control,Name="calf_r_ctrl"), bInitial=False).Transform
   )
   > &RigUnit_SetTransform(
    Item=(Type=Bone,Name="foot_l")
    Weight=1
    bInitial=False
    Value=&RigUnit_GetTransform(Item=(Type=Control,Name="foot_l_ctrl"), bInitial=False).Transform
   )
   > &RigUnit_SetTransform(
    Item=(Type=Bone,Name="foot_r")
    Weight=1
    bInitial=False
    Value=&RigUnit_GetTransform(Item=(Type=Control,Name="foot_r_ctrl"), bInitial=False).Transform
   )
   > &RigUnit_SetTransform(
    Item=(Type=Bone,Name="ball_l")
    Weight=1
    bInitial=False
    Value=&RigUnit_GetTransform(Item=(Type=Control,Name="ball_l_ctrl"), bInitial=False).Transform
   )
   > &RigUnit_SetTransform(
    Item=(Type=Bone,Name="ball_r")
    Weight=1
    bInitial=False
    Value=&RigUnit_GetTransform(Item=(Type=Control,Name="ball_r_ctrl"), bInitial=False).Transform
   )
  ]}
 New Function_ContainedGraph {nodes [
   n1 = RigVMFunctionEntryNode()
   RigVMFunction_MathQuaternionToEuler = &RigVMFunction_MathQuaternionToEuler(
    RotationOrder=XYZ
    Value=&RigUnit_GetRelativeTransformForItem(
     Child=(Type=Bone,Name="hand_l")
     Parent=(Type=Null,Name="lowerarm_l")
     bChildInitial=False
     bParentInitial=False
     Child.Name=@n1.Child
     Parent.Name=@n1.Parent
    ).RelativeTransform.Rotation
   )
   Remap = &RigVMFunction_MathDoubleRemap(
    TargetMaximum=1
    TargetMinimum=0
    Value=@RigVMFunction_MathQuaternionToEuler.Result.Z
    SourceMaximum=@n1.Source Maximum
    SourceMinimum=@n1.Source Minimum
   )
   > RigVMFunctionReturnNode(Delta Rotation=@Remap.Result)
  ]}
 New Function_ContainedGraph {nodes [
   n1 = RigVMFunctionEntryNode()
   RigVMFunction_MathQuaternionToEuler = &RigVMFunction_MathQuaternionToEuler(
    RotationOrder=XYZ
    Value=&RigUnit_GetRelativeTransformForItem(
     Child=(Type=Bone,Name="hand_l")
     Parent=(Type=Null,Name="lowerarm_l")
     bChildInitial=False
     bParentInitial=False
     Child.Name=@n1.Child
     Parent.Name=@n1.Parent
    ).RelativeTransform.Rotation
   )
   Remap = &RigVMFunction_MathDoubleRemap(
    TargetMaximum=1
    TargetMinimum=0
    Value=@RigVMFunction_MathQuaternionToEuler.Result.Y
    SourceMaximum=@n1.Source Maximum
    SourceMinimum=@n1.Source Minimum
   )
   > RigVMFunctionReturnNode(Delta Rotation=@Remap.Result)
  ]}
 New Function_ContainedGraph {nodes [
   n1 = RigVMFunctionEntryNode()
   RigVMFunction_MathQuaternionToEuler = &RigVMFunction_MathQuaternionToEuler(
    RotationOrder=XYZ
    Value=&RigUnit_GetRelativeTransformForItem(
     Child=(Type=Bone,Name="hand_l")
     Parent=(Type=Null,Name="lowerarm_l")
     bChildInitial=False
     bParentInitial=False
     Child.Name=@n1.Child
     Parent.Name=@n1.Parent
    ).RelativeTransform.Rotation
   )
   Remap = &RigVMFunction_MathDoubleRemap(
    TargetMaximum=1
    TargetMinimum=0
    Value=@RigVMFunction_MathQuaternionToEuler.Result.X
    SourceMaximum=@n1.Source Maximum
    SourceMinimum=@n1.Source Minimum
   )
   > RigVMFunctionReturnNode(Delta Rotation=@Remap.Result)
  ]}
 New Function_ContainedGraph {nodes [
   RigVMFunctionEntryNode()
   > RigVMFunction_Sequence = RigVMAggregateNode()
   n1 = &RigUnit_SphericalPoseReader(
    ActiveRegionScaleFactors=(PositiveWidth=1,NegativeWidth=1,PositiveHeight=1,NegativeHeight=1)
    ActiveRegionSize=0.05
    Debug=(bDrawDebug=false,bDraw2D=False,bDrawLocalAxes=False,DebugScale=25,DebugSegments=20,DebugThickness=0.25)
    DriverAxis=(X=1,Y=0,Z=0)
    DriverItem=(Type=Bone,Name="upperarm_l")
    FalloffRegionScaleFactors=(PositiveWidth=1,NegativeWidth=1,PositiveHeight=1,NegativeHeight=1)
    FalloffSize=0.15
    FlipHeightScaling=False
    FlipWidthScaling=False
    OptionalParentItem=(Type=Bone,Name="None")
    RotationOffset=(X=120,Y=10,Z=-30)
    ExecutePin=@RigVMFunction_Sequence.A
   )
   > &RigUnit_ModifyTransforms(
    ItemToModify=((Item=(Type=Bone,Name="dyn_coatArmpit_l"),Transform=(Rotation=(X=-0.2171,Y=-0.1508,Z=0.2803,W=-0.9228),Translation=(X=-2.8488,Y=-12.7073,Z=-10.6205),Scale3D=(X=1,Y=1,Z=1))))
    WeightMinimum=0
    Weight=@n1.OutputParam
   )
   > n2 = &RigUnit_SphericalPoseReader(
    ActiveRegionScaleFactors=(PositiveWidth=1,NegativeWidth=1,PositiveHeight=1,NegativeHeight=1)
    ActiveRegionSize=0.1
    Debug=(bDrawDebug=false,bDraw2D=False,bDrawLocalAxes=False,DebugScale=25,DebugSegments=20,DebugThickness=0.25)
    DriverAxis=(X=-1,Y=0,Z=0)
    DriverItem=(Type=Bone,Name="upperarm_r")
    FalloffRegionScaleFactors=(PositiveWidth=1,NegativeWidth=1,PositiveHeight=1,NegativeHeight=1)
    FalloffSize=0.32
    FlipHeightScaling=False
    FlipWidthScaling=False
    OptionalParentItem=(Type=Bone,Name="None")
    RotationOffset=(X=-45.0904,Y=30.3348,Z=5.5282)
   )
   > &RigUnit_ModifyTransforms(
    ItemToModify="((Item=(Type=Bone,Name=\"dyn_coatArmpit_r\"),Transform=(Rotation=(X=0.5137,Y=-0.4829,Z=-0.3034,W=0.641),Translation=(X=-10.8489,Y=-6.6276,Z=12.5357),Scale3D=(X=1,Y=1,Z=1))),(Item=(Type=Bone,Name=\"base_coatSide_r\"),Transform=(Rotation=(X=0,Y=0,Z=0,W=1),Transl ...(truncated)"
    WeightMinimum=0
    Weight=@n2.OutputParam
   )
   RigVMFunction_ControlFlowBranch = &RigVMFunction_ControlFlowBranch(ExecuteContext=@RigVMFunction_Sequence.B, Condition=RigVMVariableNode() {variable=isSkydiving_CR}.Value)
   &RigUnit_ModifyTransforms(
    ItemToModify=((Item=(Type=Bone,Name="deform_upperarm_up_l"),Transform=(Rotation=(X=0,Y=0,Z=0,W=1),Translation=(X=30,Y=0,Z=0),Scale3D=(X=1,Y=1,Z=1))))
    WeightMinimum=0
    Weight=RigVMDispatchNode(BoneName=root, Default=1, Name=deformRig_l_upperarm_up_alpha) {template="DISPATCH_RigDispatch_GetAnimAttribute(in Name,in BoneName,in Default,out Value,out Found)"}.Value
    ExecutePin=@RigVMFunction_ControlFlowBranch.True
   )
   > &RigUnit_ModifyTransforms(
    ItemToModify=((Item=(Type=Bone,Name="deform_upperarm_up_r"),Transform=(Rotation=(X=0,Y=0,Z=0,W=1),Translation=(X=-30,Y=0,Z=0),Scale3D=(X=1,Y=1,Z=1))))
    WeightMinimum=0
    Weight=RigVMDispatchNode(BoneName=root, Default=1, Name=deformRig_r_upperarm_up_alpha) {template="DISPATCH_RigDispatch_GetAnimAttribute(in Name,in BoneName,in Default,out Value,out Found)"}.Value
   )
   RigVMFunctionReturnNode(ExecuteContext=@RigVMFunction_Sequence.D)
  ]}
 New Function_ContainedGraph {nodes [
   RigVMFunctionEntryNode()
   > Delta Rotation Z = RigVMFunctionReferenceNode(Child=spine_05, Parent=spine_05_null, Source Maximum=20, Source Minimum=12)
   > &RigUnit_ModifyTransforms(
    ItemToModify=((Item=(Type=Bone,Name="base_necklace"),Transform=(Rotation=(X=0,Y=0,Z=-0.0436,W=0.999),Translation=(X=0,Y=0,Z=0),Scale3D=(X=1,Y=1,Z=1))))
    WeightMinimum=0
    Weight=@"Delta Rotation Z".Delta Rotation
   )
   > Delta Rotation X = RigVMFunctionReferenceNode(Child=spine_05, Parent=spine_05_null, Source Maximum=25, Source Minimum=10)
   > &RigUnit_ModifyTransforms(
    ItemToModify=((Item=(Type=Bone,Name="base_necklace"),Transform=(Rotation=(X=0.0076,Y=0.1735,Z=-0.043,W=0.9839),Translation=(X=0,Y=0,Z=0),Scale3D=(X=1,Y=1,Z=1))))
    WeightMinimum=0
    Weight=@"Delta Rotation X".Delta Rotation
   )
   RigVMFunction_MathVectorDistance = &RigVMFunction_MathVectorDistance(
    A=&RigUnit_GetTransform(Item=(Type=Bone,Name="spine_04"), bInitial=False).Transform.Translation
    B=&RigUnit_GetTransform(Item=(Type=Null,Name="spine_04_null"), bInitial=False).Transform.Translation
   )
   Remap = &RigVMFunction_MathDoubleRemap(SourceMaximum=4.6, SourceMinimum=0.6, TargetMaximum=1, TargetMinimum=0, Value=@RigVMFunction_MathVectorDistance.Result)
   > &RigUnit_ModifyTransforms(
    ItemToModify=((Item=(Type=Bone,Name="base_necklace"),Transform=(Rotation=(X=0,Y=0,Z=-0.0872,W=0.9962),Translation=(X=0,Y=0,Z=0),Scale3D=(X=1,Y=1,Z=1))))
    WeightMinimum=0
    Weight=@Remap.Result
   )
   > Delta Rotation Z_1 = RigVMFunctionReferenceNode(Child=hand_l, Parent=hand_l_null, Source Maximum=60, Source Minimum=30)
   > &RigUnit_ModifyTransforms(
    ItemToModify=((Item=(Type=Bone,Name="base_bracelet"),Transform=(Rotation=(X=0,Y=-0.1736,Z=0,W=0.9848),Translation=(X=0,Y=0,Z=0),Scale3D=(X=1,Y=1,Z=1))))
    WeightMinimum=0
    Weight=@"Delta Rotation Z_1".Delta Rotation
   )
   > Delta Rotation X_1 = RigVMFunctionReferenceNode(Child=hand_l, Parent=hand_l_null, Source Maximum=150)
   > &RigUnit_ModifyTransforms(
    ItemToModify=((Item=(Type=Bone,Name="base_bracelet"),Transform=(Rotation=(X=-0.0203,Y=-0.2326,Z=-0.0847,W=0.9687),Translation=(X=0,Y=0,Z=0),Scale3D=(X=1,Y=1,Z=1))))
    WeightMinimum=0
    Weight=@"Delta Rotation X_1".Delta Rotation
   )
   > RigVMFunctionReturnNode()
  ]}
 New Function_ContainedGraph {nodes [
   n1 = RigVMFunctionEntryNode()
   RigVMFunction_MathQuaternionToEuler = &RigVMFunction_MathQuaternionToEuler(
    RotationOrder=XYZ
    Value=&RigUnit_GetRelativeTransformForItem(
     Child=(Type=Bone,Name="hand_l")
     Parent=(Type=Null,Name="lowerarm_l")
     bChildInitial=False
     bParentInitial=False
     Child.Name=@n1.Child
     Parent.Name=@n1.Parent
    ).RelativeTransform.Rotation
   )
   Remap = &RigVMFunction_MathDoubleRemap(
    TargetMaximum=1
    TargetMinimum=0
    Value=@RigVMFunction_MathQuaternionToEuler.Result.X
    SourceMaximum=@n1.Source Maximum
    SourceMinimum=@n1.Source Minimum
   )
   > RigVMFunctionReturnNode(Delta Rotation=@Remap.Result)
  ]}
 RigVMFunctionLibrary {nodes [
   BeltSetup = RigVMCollapseNode(ExecuteContext=())
   Control Setup = RigVMCollapseNode(Debug Controls=False, ExecuteContext=())
   Delta Rotation X = RigVMCollapseNode(ExecuteContext=())
   Delta Rotation X_1 = RigVMCollapseNode(ExecuteContext=())
   Delta Rotation Y = RigVMCollapseNode(ExecuteContext=())
   Delta Rotation Z = RigVMCollapseNode(ExecuteContext=())
   HeelSetup = RigVMCollapseNode(ExecuteContext=())
   SurfaceConstraints = RigVMCollapseNode(ExecuteContext=())
   accessoriesFix = RigVMCollapseNode(ExecuteContext=())
   coatFix = RigVMCollapseNode(ExecuteContext=())
  ]}
 RigVMFunction_Sequence_1_ContainedGraph {nodes [
   n1 = RigVMFunctionEntryNode()
   RigVMFunction_Sequence = &RigVMFunction_Sequence(ExecuteContext=@n1.ExecuteContext)
   n2 = &RigVMFunction_Sequence(ExecuteContext=@RigVMFunction_Sequence.B)
   RigVMFunctionReturnNode(A=@RigVMFunction_Sequence.A, B=@n2.A, C=@n2.B)
  ]}
 RigVMFunction_Sequence_1_ContainedGraph {nodes [
   n1 = RigVMFunctionEntryNode()
   RigVMFunction_Sequence = &RigVMFunction_Sequence(ExecuteContext=@n1.ExecuteContext)
   n2 = &RigVMFunction_Sequence(ExecuteContext=@RigVMFunction_Sequence.B)
   RigVMFunctionReturnNode(A=@RigVMFunction_Sequence.A, B=@n2.A, C=@n2.B)
  ]}
 RigVMFunction_Sequence_1_ContainedGraph {nodes [
   n1 = RigVMFunctionEntryNode()
   RigVMFunction_Sequence = &RigVMFunction_Sequence(ExecuteContext=@n1.ExecuteContext)
   n2 = &RigVMFunction_Sequence(ExecuteContext=@RigVMFunction_Sequence.B)
   n3 = &RigVMFunction_Sequence(ExecuteContext=@n2.B)
   RigVMFunctionReturnNode(A=@RigVMFunction_Sequence.A, B=@n2.A, C=@n3.A, D=@n3.B)
  ]}
 RigVMModel {nodes [
   RigUnit_BeginExecution = &RigUnit_BeginExecution()
   RigVMFunction_Sequence = &RigVMFunction_Sequence(ExecuteContext=@RigUnit_BeginExecution.ExecutePin)
   Control Setup = RigVMFunctionReferenceNode(Debug Controls=false, ExecuteContext=@RigVMFunction_Sequence.A)
   /Script/FortniteGame.RigUnit_FortDeformRig(
    Bool=(bUpperArm=true,bElbow=true,bWrist=true,bGroin=true,bGlute=true,bKnee=true,bPec=true,bBicep=false,bLat=false)
    Curve="(bLUpperArmUpAlpha=true,bRUpperArmUpAlpha=true,bLUpperArmFrAlpha=false,bRUpperArmFrAlpha=false,bLUpperArmBkAlpha=False,bRUpperArmBkAlpha=False,bLWristUpAlpha=False,bRWristUpAlpha=False,bLWristDnAlpha=False,bRWristDnAlpha=False,bLKneeAlpha=False,bRKneeAlpha ...(truncated)"
    Debug=(bUpperArmDebug=False,bWristDebug=False,bLatDebug=False,bGroinGluteDebug=false)
    Value="(UpperArmUp=(Rotation=(X=0.0,Y=-0.0,Z=0.0,W=1.0),Translation=(X=10,Y=0,Z=8),Scale3D=(X=0.70,Y=1.0,Z=1.0)),UpperArmFwdFr=(Rotation=(X=0.0,Y=-0.0,Z=0.0,W=1.0),Translation=(X=0.0,Y=10.0,Z=0.0),Scale3D=(X=0.50,Y=1.0,Z=1.0)),UpperArmFwdBk=(Rotation=(X=0,Y=0,Z=0 ...(truncated)"
    ExecutePin=@RigVMFunction_Sequence.B
   ) {is=struct RigVMUnitNode}
   > SurfaceConstraints = RigVMFunctionReferenceNode()
   > HeelSetup = RigVMFunctionReferenceNode()
   > BeltSetup = RigVMFunctionReferenceNode()
   > accessoriesFix = RigVMFunctionReferenceNode()
   > coatFix = RigVMFunctionReferenceNode()
  ]}
 RigVMModel Construction Event Graph {nodes [
   n1 = &RigUnit_PrepareForExecution()
   Multiply = &RigVMFunction_MathTransformMul(
    A=(Rotation=(X=0,Y=0,Z=0,W=1),Translation=(X=0,Y=0,Z=-10),Scale3D=(X=1,Y=1,Z=1))
    B=&RigUnit_GetRelativeTransformForItem(Child=(Type=Bone,Name="heelFoot_l"), Parent=(Type=Bone,Name="foot_l"), bChildInitial=true, bParentInitial=true).RelativeTransform
   ) {template="Multiply::Execute(in A,in B,out Result)"}
   &RigUnit_HierarchyAddNull(Name=metaTarsoRet_l_upV_null, Parent=(Type=Bone,Name="foot_l"), Space=LocalSpace, Transform=@Multiply.Result, ExecutePin=@n1.ExecutePin)
   > &RigUnit_HierarchyAddNull(
    Name=metaTarsoRet_r_upV_null
    Parent=(Type=Bone,Name="foot_r")
    Space=LocalSpace
    Transform=&RigVMFunction_MathTransformMul(
     A=(Rotation=(X=0,Y=0,Z=0,W=1),Translation=(X=0,Y=0,Z=10),Scale3D=(X=1,Y=1,Z=1))
     B=&RigUnit_GetRelativeTransformForItem(Child=(Type=Bone,Name="heelFoot_r"), Parent=(Type=Bone,Name="foot_r"), bChildInitial=true, bParentInitial=true).RelativeTransform
    ) {template="Multiply::Execute(in A,in B,out Result)"}.Result
   )
  ]}
]
hierarchy [
 root {loc=<0 0 0>, rot=<0 0 0 1>}
 pelvis {parent=root, loc=<0 -0.5301 98.0571>, rot=<0.0127 0.707 -0.0127 -0.707>}
 spine_01 {parent=pelvis, loc=<6.1208 -0.9328 0>, rot=<0 0 0.034 -0.9994>}
 spine_02 {parent=spine_01, loc=<6.5997 -0.2095 0>, rot=<0 0 0.0238 -0.9997>}
 spine_03 {parent=spine_02, loc=<6.4914 -0.1237 0>, rot=<0 0 0.0149 -0.9999>}
 spine_04 {parent=spine_03, loc=<6.9598 0.0794 0>, rot=<0 0 -0.0132 -0.9999>}
 spine_05 {parent=spine_04, loc=<6.999 0 0>, rot=<0 0 -0.0015 -1>}
 neck_01 {parent=spine_05, loc=<11.3138 0.1529 0>, rot=<0 0 -0.2003 -0.9797>}
 neck_02 {parent=neck_01, loc=<4.5627 0 0>, rot=<0 0 0.0504 -0.9987>}
 head {parent=neck_02, loc=<5.3914 0 0>, rot=<0 0 0.115 -0.9934>}
 clavicle_l {parent=spine_05, loc=<5.4831 -0.1053 -3.2946>, rot=<0.0465 -0.7289 -0.009 -0.6829>}
 upperarm_l {parent=clavicle_l, loc=<9.8658 -1.2136 -0.4105>, rot=<0.0164 -0.3245 -0.0269 -0.9454>}
 lowerarm_l {parent=upperarm_l, loc=<26.465 -0.8345 0.259>, rot=<0.0036 0.0161 -0.2298 -0.9731>}
 hand_l {parent=lowerarm_l, loc=<22.7355 -0.3203 -0.6105>, rot=<-0.6784 -0.0223 0.0735 -0.7307>}
 thumb_01_l {parent=hand_l, loc=<1.3118 -0.3191 -2.459>, rot=<0.6546 -0.4042 0.0664 -0.6354>}
 thumb_02_l {parent=thumb_01_l, loc=<5.9336 -0.0448 0.0401>, rot=<-0.0037 0.0213 0.298 -0.9543>}
 thumb_03_l {parent=thumb_02_l, loc=<2.2223 0.3682 -0.0907>, rot=<0.0118 -0.013 0.0066 -0.9998>}
 index_metacarpal_l {parent=hand_l, loc=<1.6535 0.0183 -2.96>, rot=<0.0158 -0.0124 -0.0173 -0.9996>}
 index_01_l {parent=index_metacarpal_l, loc=<8.3008 0.0357 -0.0524>, rot=<-0.0044 0.0485 0.0924 -0.9945>}
 index_02_l {parent=index_01_l, loc=<3.3587 -0.3004 0.0452>, rot=<-0.002 0.0027 0.3225 -0.9466>}
 index_03_l {parent=index_02_l, loc=<2.5578 -0.123 -0.0252>, rot=<-0.002 0.0067 0.2558 -0.9667>}
 middle_metacarpal_l {parent=hand_l, loc=<2.1856 -0.1725 -0.5984>, rot=<0.006 0.0168 -0.0329 -0.9993>}
 middle_01_l {parent=middle_metacarpal_l, loc=<7.3275 -0.2447 0.0129>, rot=<-0.0208 0.0042 0.0858 -0.9961>}
 middle_02_l {parent=middle_01_l, loc=<4.0123 0.1401 0.1175>, rot=<-0.0218 0.0258 0.3512 -0.9357>}
 middle_03_l {parent=middle_02_l, loc=<2.6077 0.0534 0.1373>, rot=<0.0082 0.0188 0.1708 -0.9851>}
 ring_metacarpal_l {parent=hand_l, loc=<2.2074 -0.3625 1.3638>, rot=<-0.0284 0.0722 -0.0458 -0.9959>}
 ring_01_l {parent=ring_metacarpal_l, loc=<7.0295 -0.6197 -0.1446>, rot=<-0.0489 -0.0252 0.1892 -0.9804>}
 ring_02_l {parent=ring_01_l, loc=<3.6174 0.4647 -0.1027>, rot=<0 0 0.2806 -0.9598>}
 ring_03_l {parent=ring_02_l, loc=<2.4152 0.114 0.0199>, rot=<0 0 0.1189 -0.9929>}
 pinky_metacarpal_l {parent=hand_l, loc=<1.7616 -0.9753 3.3814>, rot=<-0.0397 0.0814 -0.069 -0.9935>}
 pinky_01_l {parent=pinky_metacarpal_l, loc=<7.1709 -0.4831 -0.1094>, rot=<-0.0654 -0.0035 0.2242 -0.9723>}
 pinky_02_l {parent=pinky_01_l, loc=<2.8274 0.0893 -0.1043>, rot=<0.0037 0.0011 0.2854 -0.9584>}
 pinky_03_l {parent=pinky_02_l, loc=<2.2046 0.0307 -0.0078>, rot=<-0.0002 0.0098 0.1445 -0.9895>}
 deform_wrist_dn_base_l {parent=hand_l, loc=<0 -1.5326 0>, rot=<0 0 0 -1>}
 deform_wrist_dn_l {parent=deform_wrist_dn_base_l, loc=<0 0 0>, rot=<0 0 0 -1>}
 deform_wrist_up_base_l {parent=hand_l, loc=<0 2.3009 0>, rot=<0 0 0 -1>}
 deform_wrist_up_l {parent=deform_wrist_up_base_l, loc=<0 0 0>, rot=<0 0 0 -1>}
 lowerarm_twist_02_l {parent=lowerarm_l, loc=<10.6707 0 0>, rot=<0 0 0 1>}
 base_outSleeveFur_l {parent=lowerarm_twist_02_l, loc=<-2.0141 -7.045 0.0908>, rot=<0.0021 -0.0224 0.0923 -0.9955>}
 dyn_outSleeveFur_l {parent=base_outSleeveFur_l, loc=<0 0 0>, rot=<0 0 0 -1>}
 base_inSleeveFur_l {parent=lowerarm_twist_02_l, loc=<-0.2803 7.5094 0.0126>, rot=<0.9997 0.0011 -0.0225 0>}
 dyn_inSleeveFur_l {parent=base_inSleeveFur_l, loc=<0 0 0>, rot=<0 0 0 1>}
 base_bottomSleeveFur_l {parent=lowerarm_twist_02_l, loc=<-1.4056 -0.0023 -7.7796>, rot=<-0.7052 -0.0521 0.0506 -0.7053>}
 dyn_bottomSleeveFur_l {parent=base_bottomSleeveFur_l, loc=<0 0 0>, rot=<0 0 0 1>}
 base_topSleeveFur_l {parent=lowerarm_twist_02_l, loc=<-1.0008 -0.0029 7.5055>, rot=<0.7069 0.0155 0.014 -0.707>}
 dyn_topSleeveFur_l {parent=base_topSleeveFur_l, loc=<0 0 0>, rot=<0 0 0 1>}
 lowerarm_twist_01_l {parent=lowerarm_l, loc=<18.316 0 0>, rot=<0 0 0 1>}
 deform_elbow_in_base_l {parent=lowerarm_l, loc=<0.1834 3.6062 -0.2572>, rot=<0 0 0 -1>}
 deform_elbow_in_l {parent=deform_elbow_in_base_l, loc=<0 0 0>, rot=<0 0 0 -1>}
 deform_elbow_out_base_l {parent=lowerarm_l, loc=<0 -4 0>, rot=<0 0 0 1>}
 deform_elbow_out_l {parent=deform_elbow_out_base_l, loc=<0 0 0>, rot=<0 0 0 1>}
 hand_l_null {parent=lowerarm_l, type=Null, loc=<22.7355 -0.3203 -0.6106>, rot=<-0.6784 -0.0223 0.0735 -0.7307>}
 base_bracelet {parent=lowerarm_l, loc=<21.4082 -0.3937 -0.1239>, rot=<0 0 0 -1>}
 dyn_bracelet {parent=base_bracelet, loc=<0 0 0>, rot=<0 0 0 -1>}
 upperarm_twist_01_l {parent=upperarm_l, loc=<7.2135 -0.1134 0>, rot=<0 0 0 1>}
 deform_upperarm_bk_base_l {parent=upperarm_twist_01_l, loc=<-6.4053 -4.3746 0.314>, rot=<0 0 0 -1>}
 deform_upperarm_bk_l {parent=deform_upperarm_bk_base_l, loc=<0 0 0>, rot=<0 0 0 -1>}
 deform_upperarm_fr_base_l {parent=upperarm_twist_01_l, loc=<-5.6312 5.092 0.3354>, rot=<0 0 0 -1>}
 deform_upperarm_fr_l {parent=deform_upperarm_fr_base_l, loc=<0 0 0>, rot=<0 0 0 -1>}
 deform_upperarm_up_base_l {parent=upperarm_twist_01_l, loc=<-8.891 0.4825 3.9517>, rot=<0 0 0 -1>}
 deform_upperarm_up_l {parent=deform_upperarm_up_base_l, loc=<0 0 0>, rot=<0 0 0 -1>}
 upperarm_twist_02_l {parent=upperarm_l, loc=<19.7998 -0.1134 0>, rot=<0 0 0 1>}
 clavicle_r {parent=spine_05, loc=<5.4831 -0.1053 3.2946>, rot=<0.7289 0.0465 -0.6829 0.009>}
 upperarm_r {parent=clavicle_r, loc=<-9.8658 1.2136 0.4105>, rot=<0.0164 -0.3245 -0.0269 -0.9454>}
 lowerarm_r {parent=upperarm_r, loc=<-26.465 0.8345 -0.259>, rot=<0.0036 0.0161 -0.2298 -0.9731>}
 hand_r {parent=lowerarm_r, loc=<-22.7355 0.3203 0.6106>, rot=<-0.6784 -0.0223 0.0735 -0.7307>}
 thumb_01_r {parent=hand_r, loc=<-1.3118 0.3191 2.459>, rot=<0.6546 -0.4042 0.0664 -0.6354>}
 thumb_02_r {parent=thumb_01_r, loc=<-5.9336 0.0448 -0.0401>, rot=<-0.0037 0.0213 0.298 -0.9543>}
 thumb_03_r {parent=thumb_02_r, loc=<-2.2223 -0.3682 0.0907>, rot=<0.0118 -0.013 0.0066 -0.9998>}
 index_metacarpal_r {parent=hand_r, loc=<-1.6535 -0.0183 2.96>, rot=<0.0158 -0.0124 -0.0173 -0.9996>}
 index_01_r {parent=index_metacarpal_r, loc=<-8.3008 -0.0357 0.0524>, rot=<-0.0044 0.0485 0.0924 -0.9945>}
 index_02_r {parent=index_01_r, loc=<-3.3587 0.3004 -0.0452>, rot=<-0.002 0.0027 0.3225 -0.9466>}
 index_03_r {parent=index_02_r, loc=<-2.5578 0.123 0.0252>, rot=<-0.002 0.0067 0.2558 -0.9667>}
 middle_metacarpal_r {parent=hand_r, loc=<-2.1856 0.1725 0.5984>, rot=<0.006 0.0168 -0.0329 -0.9993>}
 middle_01_r {parent=middle_metacarpal_r, loc=<-7.3275 0.2447 -0.0129>, rot=<-0.0208 0.0042 0.0858 -0.9961>}
 middle_02_r {parent=middle_01_r, loc=<-4.0123 -0.1401 -0.1175>, rot=<-0.0218 0.0258 0.3512 -0.9357>}
 middle_03_r {parent=middle_02_r, loc=<-2.6077 -0.0534 -0.1373>, rot=<0.0082 0.0188 0.1708 -0.9851>}
 ring_metacarpal_r {parent=hand_r, loc=<-2.2074 0.3625 -1.3638>, rot=<-0.0284 0.0722 -0.0458 -0.9959>}
 ring_01_r {parent=ring_metacarpal_r, loc=<-7.0295 0.6197 0.1446>, rot=<-0.0489 -0.0252 0.1892 -0.9804>}
 ring_02_r {parent=ring_01_r, loc=<-3.6174 -0.4647 0.1027>, rot=<0 0 0.2806 -0.9598>}
 ring_03_r {parent=ring_02_r, loc=<-2.4152 -0.114 -0.0199>, rot=<0 0 0.1189 -0.9929>}
 pinky_metacarpal_r {parent=hand_r, loc=<-1.7616 0.9753 -3.3814>, rot=<-0.0397 0.0814 -0.069 -0.9935>}
 pinky_01_r {parent=pinky_metacarpal_r, loc=<-7.1709 0.4831 0.1094>, rot=<-0.0654 -0.0035 0.2242 -0.9723>}
 pinky_02_r {parent=pinky_01_r, loc=<-2.8274 -0.0893 0.1043>, rot=<0.0037 0.0011 0.2854 -0.9584>}
 pinky_03_r {parent=pinky_02_r, loc=<-2.2046 -0.0307 0.0078>, rot=<-0.0002 0.0098 0.1445 -0.9895>}
 deform_wrist_dn_base_r {parent=hand_r, loc=<-0.0001 1.5328 0>, rot=<0 0 0 -1>}
 deform_wrist_dn_r {parent=deform_wrist_dn_base_r, loc=<0 0 0>, rot=<0 0 0 -1>}
 deform_wrist_up_base_r {parent=hand_r, loc=<-0.0002 -2.3007 -0.0001>, rot=<0 0 0 -1>}
 deform_wrist_up_r {parent=deform_wrist_up_base_r, loc=<0 0 0>, rot=<0 0 0 -1>}
 lowerarm_twist_02_r {parent=lowerarm_r, loc=<-10.6707 0 0>, rot=<0 0 0 -1>}
 base_inSleeveFur_r {parent=lowerarm_twist_02_r, loc=<0.3097 -7.5247 0.0152>, rot=<0.0011 -0.9997 0 0.0225>}
 dyn_inSleeveFur_r {parent=base_inSleeveFur_r, loc=<0 0 0>, rot=<0 0 0 1>}
 base_topSleeveFur_r {parent=lowerarm_twist_02_r, loc=<1.03 -0.0123 -7.4775>, rot=<-0.0155 0.7069 0.707 0.014>}
 dyn_topSleeveFur_r {parent=base_topSleeveFur_r, loc=<0 0 0>, rot=<0 0 0 1>}
 base_outSleeveFur_r {parent=lowerarm_twist_02_r, loc=<2.0434 7.0298 -0.0629>, rot=<0.0224 0.0021 0.9955 0.0923>}
 dyn_outSleeveFur_r {parent=base_outSleeveFur_r, loc=<0 0 0>, rot=<0 0 0 1>}
 base_bottomSleeveFur_r {parent=lowerarm_twist_02_r, loc=<1.4349 -0.013 7.8074>, rot=<0.0521 -0.7052 0.7053 0.0506>}
 dyn_bottomSleeveFur_r {parent=base_bottomSleeveFur_r, loc=<0 0 0>, rot=<0 0 0 1>}
 lowerarm_twist_01_r {parent=lowerarm_r, loc=<-18.316 0 0>, rot=<0 0 0 -1>}
 deform_elbow_in_base_r {parent=lowerarm_r, loc=<-0.1834 -3.6062 0.2572>, rot=<0 0 0 -1>}
 deform_elbow_in_r {parent=deform_elbow_in_base_r, loc=<0 0 0>, rot=<0 0 0 -1>}
 deform_elbow_out_base_r {parent=lowerarm_r, loc=<0 4 -0.0001>, rot=<0 0 0 -1>}
 deform_elbow_out_r {parent=deform_elbow_out_base_r, loc=<0 0 0>, rot=<0 0 0 1>}
 upperarm_twist_01_r {parent=upperarm_r, loc=<-7.2135 0.1134 0>, rot=<0 0 0 -1>}
 deform_upperarm_bk_base_r {parent=upperarm_twist_01_r, loc=<6.4053 4.3746 -0.314>, rot=<0 0 0 -1>}
 deform_upperarm_bk_r {parent=deform_upperarm_bk_base_r, loc=<0 0 0>, rot=<0 0 0 -1>}
 deform_upperarm_fr_base_r {parent=upperarm_twist_01_r, loc=<5.6312 -5.092 -0.3354>, rot=<0 0 0 -1>}
 deform_upperarm_fr_r {parent=deform_upperarm_fr_base_r, loc=<0 0 0>, rot=<0 0 0 -1>}
 deform_upperarm_up_base_r {parent=upperarm_twist_01_r, loc=<8.891 -0.4825 -3.9517>, rot=<0 0 0 -1>}
 deform_upperarm_up_r {parent=deform_upperarm_up_base_r, loc=<0 0 0>, rot=<0 0 0 -1>}
 upperarm_twist_02_r {parent=upperarm_r, loc=<-19.7998 0.1134 0>, rot=<0 0 0 -1>}
 attach_backpack {parent=spine_05, loc=<0.0504 -13.6199 0.0265>, rot=<-0.0087 -0.7072 0.0101 -0.7069>}
 attach_cape {parent=spine_05, loc=<9.1283 -11.8451 0.0191>, rot=<0.0681 -0.7262 -0.0029 -0.6841>}
 base_necklaceCharm {parent=spine_05, loc=<2.1697 8.5531 0.3292>, rot=<0 0 0.9992 -0.0401>}
 dyn_necklaceCharm {parent=base_necklaceCharm, loc=<0 0 0>, rot=<0 0 0 1>}
 Backpack_BR {parent=spine_05, type=Null, loc=<-0.1941 -16.4845 0.0008>, rot=<-0.0458 -0.7071 0.046 -0.7041>}
 cape {parent=spine_05, type=Null, loc=<8.6777 -15.567 0.0241>, rot=<-0.0042 0.7061 0.0042 0.7081>}
 base_coatFur_02_r {parent=spine_05, loc=<8.9357 6.4269 13.8317>, rot=<0.6463 -0.2869 0.6463 0.2869>}
 dyn_coatFur_02_r {parent=base_coatFur_02_r, loc=<0 0 0>, rot=<0 0 0 1>}
 base_coatFur_03_r {parent=spine_05, loc=<13.0677 -2.8289 12.9103>, rot=<0.5982 -0.5386 0.441 0.3971>}
 dyn_coatFur_03_r {parent=base_coatFur_03_r, loc=<0 0 0>, rot=<0 0 0 1>}
 base_coatFur_04_r {parent=spine_05, loc=<12.1152 -9.1875 6.6829>, rot=<0.686 -0.688 0.1671 0.1676>}
 dyn_coatFur_04_r {parent=base_coatFur_04_r, loc=<0 0 0>, rot=<0 0 0 1>}
 base_coatFur_01_r {parent=spine_05, loc=<1.4736 10.5717 12.7881>, rot=<0.6982 -0.1122 0.6982 0.1122>}
 dyn_coatFur_01_r {parent=base_coatFur_01_r, loc=<0 0 0>, rot=<0 0 0 1>}
 base_coatFur_c {parent=spine_05, loc=<11.2107 -11.7236 0>, rot=<0.7071 -0.7071 0 0>}
 dyn_coatFur_c {parent=base_coatFur_c, loc=<0 0 0>, rot=<0 0 0 1>}
 base_coatFur_04_l {parent=spine_05, loc=<12.115 -9.1875 -6.6829>, rot=<0.686 -0.688 -0.1671 -0.1676>}
 dyn_coatFur_04_l {parent=base_coatFur_04_l, loc=<0 0 0>, rot=<0 0 0 1>}
 base_coatFur_03_l {parent=spine_05, loc=<13.0673 -2.829 -12.9103>, rot=<0.5982 -0.5386 -0.441 -0.3971>}
 dyn_coatFur_03_l {parent=base_coatFur_03_l, loc=<0 0 0>, rot=<0 0 0 1>}
 base_coatFur_02_l {parent=spine_05, loc=<8.9354 6.4268 -13.8317>, rot=<0.6463 -0.2869 -0.6463 -0.2869>}
 dyn_coatFur_02_l {parent=base_coatFur_02_l, loc=<0 0 0>, rot=<0 0 0 1>}
 base_coatFur_01_l {parent=spine_05, loc=<1.4732 10.5717 -12.7881>, rot=<0.6982 -0.1122 -0.6982 -0.1122>}
 dyn_coatFur_01_l {parent=base_coatFur_01_l, loc=<0 0 0>, rot=<0 0 0 1>}
 base_coatVolume01_l {parent=spine_05, loc=<1.561 9.379 -14.72>, rot=<0 0 0 1>}
 dyn_coatVolume01_l {parent=base_coatVolume01_l, loc=<0 0 0>, rot=<0 0 0 1>}
 base_coatVolume01_r {parent=spine_05, loc=<1.5614 9.379 14.7196>, rot=<0 0 0 1>}
 dyn_coatVolume01_r {parent=base_coatVolume01_r, loc=<0 0 0>, rot=<0 0 0 1>}
 base_necklace {parent=spine_05, loc=<4.4215 6.7715 0.0002>, rot=<0 0 -0.6683 -0.7439>}
 dyn_necklace {parent=base_necklace, loc=<0 0 0>, rot=<0 0 0 -1>}
 deform_pec_aim_base_l {parent=spine_04, loc=<3.2456 9.3309 -1.2237>, rot=<0.0603 -0.4033 0.3318 -0.8507>}
 deform_pec_aim_l {parent=deform_pec_aim_base_l, loc=<0 0 0>, rot=<0 0 0 -1>}
 deform_pec_base_l {parent=deform_pec_aim_l, loc=<9.3665 0 0>, rot=<0 0 0 -1>}
 deform_pec_l {parent=deform_pec_base_l, loc=<0.0001 0 0>, rot=<0 0 0 -1>}
 deform_pec_aim_base_r {parent=spine_04, loc=<3.2452 9.3309 1.2237>, rot=<-0.4033 -0.0603 0.8507 0.3318>}
 deform_pec_aim_r {parent=deform_pec_aim_base_r, loc=<0 0 0>, rot=<0 0 0 -1>}
 deform_pec_base_r {parent=deform_pec_aim_r, loc=<-9.3667 -0.0003 -0.0004>, rot=<0 0 0 -1>}
 deform_pec_r {parent=deform_pec_base_r, loc=<-0.0001 0 0>, rot=<0 0 0 -1>}
 base_coatFront_l {parent=spine_04, loc=<-2.8334 6.8372 -12.8954>, rot=<-0.0239 0.5736 0.8181 -0.0341>}
 dyn_coatFront_l {parent=base_coatFront_l, loc=<0 0 0>, rot=<0 0 0 1>}
 base_coatSide_l {parent=spine_04, loc=<-3.423 -2.3094 -13.8521>, rot=<0.0322 -0.7726 -0.6335 0.0264>}
 dyn_coatSide_l {parent=base_coatSide_l, loc=<0 0 0>, rot=<0 0 0 1>}
 base_coatSideBack_l {parent=spine_04, loc=<-2.8506 -9.1719 -8.6126>, rot=<0.0404 -0.9698 -0.2405 0.01>}
 dyn_coatSideBack_l {parent=base_coatSideBack_l, loc=<0 0 0>, rot=<0 0 0 1>}
 base_coatBack {parent=spine_04, loc=<-2.7541 -10.3293 0>, rot=<0.0416 -0.9991 0 0>}
 dyn_coatBack {parent=base_coatBack, loc=<0 0 0>, rot=<0 0 0 1>}
 base_coatSideBack_r {parent=spine_04, loc=<-2.8511 -9.1719 8.6126>, rot=<0.0404 -0.9698 0.2405 -0.01>}
 dyn_coatSideBack_r {parent=base_coatSideBack_r, loc=<0 0 0>, rot=<0 0 0 1>}
 base_coatSide_r {parent=spine_04, loc=<-3.4234 -2.3094 13.8521>, rot=<0.0322 -0.7726 0.6335 -0.0264>}
 dyn_coatSide_r {parent=base_coatSide_r, loc=<0 0 0>, rot=<0 0 0 1>}
 base_coatFront_r {parent=spine_04, loc=<-2.8336 6.8372 12.8954>, rot=<0.0239 -0.5736 0.8181 -0.0341>}
 dyn_coatFront_r {parent=base_coatFront_r, loc=<0 0 0>, rot=<0 0 0 1>}
 base_coatArmpit_l {parent=spine_04, loc=<0.2915 -2.8803 -16.0036>, rot=<0.0322 -0.7726 -0.6335 0.0264>}
 dyn_coatArmpit_l {parent=base_coatArmpit_l, loc=<0 0 0>, rot=<0 0 0 1>}
 base_coatArmpit_r {parent=spine_04, loc=<0.2915 -2.8803 16.004>, rot=<0.0264 -0.6335 0.7726 -0.0322>}
 dyn_coatArmpit_r {parent=base_coatArmpit_r, loc=<0 0 0>, rot=<0 0 0 1>}
 spine_04_null {parent=spine_03, type=Null, loc=<6.9598 0.0794 0>, rot=<0 0 -0.0132 -0.9999>}
 thigh_l {parent=pelvis, loc=<-1.0855 -0.2109 -8.7053>, rot=<0.0009 0.0355 0.0155 -0.9993>}
 calf_l {parent=thigh_l, loc=<-42.5284 -0.0019 -0.7297>, rot=<0.0006 -0.0161 -0.0391 -0.9991>}
 foot_l {parent=calf_l, loc=<-46.7401 -2.5014 -0.592>, rot=<0.0427 -0.0238 0.0353 -0.9982>}
 ball_l {parent=foot_l, loc=<-3.5658 10.8066 -0.1544>, rot=<0.0001 -0.0001 -0.7063 -0.7079>}
 calf_twist_02_l {parent=calf_l, loc=<-19.1546 -1.7943 -0.2638>, rot=<0.0237 0.0095 -0.0035 -0.9997>}
 calf_twist_01_l {parent=calf_l, loc=<-35.8483 -1.9197 -0.5794>, rot=<0.0237 0.0095 -0.0035 -0.9997>}
 base_heelBullet_l {parent=calf_twist_01_l, loc=<0.6914 -5.0055 -0.4052>, rot=<-0.0445 -0.9982 -0.0257 0.0299>}
 dyn_heelBullet_l {parent=base_heelBullet_l, loc=<0 0 0>, rot=<0 0 0 1>}
 deform_knee_in_base_l {parent=calf_l, loc=<0.8031 -5.0231 0.2456>, rot=<0 0 0 -1>}
 deform_knee_in_l {parent=deform_knee_in_base_l, loc=<0 0 0>, rot=<0 0 0 -1>}
 deform_knee_out_base_l {parent=calf_l, loc=<0 4 0>, rot=<0 0 0 -1>}
 deform_knee_out_l {parent=deform_knee_out_base_l, loc=<0 0 0>, rot=<0 0 0 1>}
 heelFoot_l {parent=calf_l, loc=<-39.3117 -2.099 -0.5224>, rot=<0.0432 -0.7194 0.0986 -0.6862>}
 heelBall_l {parent=heelFoot_l, loc=<0 6.5639 -13.8742>, rot=<0 0 0 1>}
 heelFoot_l_null {parent=calf_l, type=Null, loc=<-39.3117 -2.099 -0.5224>, rot=<0.0432 -0.7194 0.0986 -0.6862>}
 thigh_twist_01_l {parent=thigh_l, loc=<-27.952 0 0>, rot=<0.0013 0 0 -1>}
 deform_groin_base_l {parent=thigh_l, loc=<0.4801 5.6776 -0.2486>, rot=<0 0 0 -1>}
 deform_groin_l {parent=deform_groin_base_l, loc=<0 0 0>, rot=<0 0 0 -1>}
 thigh_r {parent=pelvis, loc=<-1.0855 -0.2109 8.7053>, rot=<0.0355 -0.0009 0.9993 0.0155>}
 calf_r {parent=thigh_r, loc=<42.5284 0.0019 0.7297>, rot=<0.0006 -0.0161 -0.0391 -0.9991>}
 foot_r {parent=calf_r, loc=<46.7401 2.5014 0.592>, rot=<0.0427 -0.0238 0.0353 -0.9982>}
 ball_r {parent=foot_r, loc=<3.5658 -10.8066 0.1544>, rot=<0.0001 -0.0001 -0.7063 -0.7079>}
 calf_twist_02_r {parent=calf_r, loc=<19.1546 1.7943 0.2638>, rot=<0.0237 0.0095 -0.0035 -0.9997>}
 calf_twist_01_r {parent=calf_r, loc=<35.8483 1.9197 0.5794>, rot=<0.0237 0.0095 -0.0035 -0.9997>}
 base_heelBullet_r {parent=calf_twist_01_r, loc=<-0.6982 4.9464 0.406>, rot=<0.9982 -0.0445 -0.0299 -0.0257>}
 dyn_heelBullet_r {parent=base_heelBullet_r, loc=<0 0 0>, rot=<0 0 0 1>}
 deform_knee_in_base_r {parent=calf_r, loc=<-0.8031 5.0231 -0.2457>, rot=<0 0 0 -1>}
 deform_knee_in_r {parent=deform_knee_in_base_r, loc=<0 0 0>, rot=<0 0 0 -1>}
 deform_knee_out_base_r {parent=calf_r, loc=<0 -4 0>, rot=<0 0 0 -1>}
 deform_knee_out_r {parent=deform_knee_out_base_r, loc=<0 0 0>, rot=<0 0 0 1>}
 heelFoot_r {parent=calf_r, loc=<39.3117 2.099 0.5225>, rot=<-0.6862 0.0986 0.7194 -0.0432>}
 heelBall_r {parent=heelFoot_r, loc=<0 6.5639 -13.8742>, rot=<0 0 0 1>}
 heelFoot_r_null {parent=calf_r, type=Null, loc=<39.3117 2.099 0.5225>, rot=<-0.6862 0.0986 0.7194 -0.0432>}
 thigh_twist_01_r {parent=thigh_r, loc=<27.952 0 0>, rot=<0.0013 0 0 -1>}
 deform_groin_base_r {parent=thigh_r, loc=<-0.4801 -5.6776 0.2486>, rot=<0 0 0 -1>}
 deform_groin_r {parent=deform_groin_base_r, loc=<0 0 0>, rot=<0 0 0 -1>}
 deform_glute_base_l {parent=pelvis, loc=<-7.3466 -9.0213 -9.1245>, rot=<0.0172 0.031 -0.4477 -0.8935>}
 deform_glute_l {parent=deform_glute_base_l, loc=<0 0 0>, rot=<0 0 0 1>}
 deform_glute_base_r {parent=pelvis, loc=<-7.3464 -9.0213 9.1245>, rot=<0.031 -0.0172 0.8935 -0.4477>}
 deform_glute_r {parent=deform_glute_base_r, loc=<0 0 0>, rot=<0 0 0 1>}
 base_handkerchief {parent=pelvis, loc=<4.0226 0.9368 13.7263>, rot=<-0.0127 -0.707 0.707 0.0127>}
 dyn_handkerchief {parent=base_handkerchief, loc=<0 0 0>, rot=<0 0 0 1>}
 base_handkerchiefBottom {parent=dyn_handkerchief, loc=<0.8186 -1.0544 -1.3363>, rot=<-0.2124 -0.0496 0.081 -0.9726>}
 dyn_handkerchiefBottom_01 {parent=base_handkerchiefBottom, loc=<0 0 0>, rot=<0 0 0 1>}
 dyn_handkerchiefBottom_02 {parent=dyn_handkerchiefBottom_01, loc=<7.6375 0 0>, rot=<0 0 0 1>}
 dyn_handkerchiefBottom_03 {parent=dyn_handkerchiefBottom_02, loc=<7.6375 0 0>, rot=<0 0 0 1>}
 base_handkerchiefTop {parent=dyn_handkerchief, loc=<0.8186 -1.2061 1.5057>, rot=<-0.0101 0.044 0.2245 -0.9734>}
 dyn_handkerchiefTop_01 {parent=base_handkerchiefTop, loc=<0 0 0>, rot=<0 0 0 1>}
 dyn_handkerchiefTop_02 {parent=dyn_handkerchiefTop_01, loc=<4.6866 0 0>, rot=<0 0 0 1>}
 dyn_handkerchiefTop_03 {parent=dyn_handkerchiefTop_02, loc=<4.6866 0 0>, rot=<0 0 0 1>}
 thigh_l_null {parent=pelvis, type=Null, loc=<-1.0855 -0.2109 -8.7053>, rot=<0.0009 0.0355 0.0155 -0.9993>}
 thigh_r_null {parent=pelvis, type=Null, loc=<-1.0855 -0.2109 8.7053>, rot=<0.0355 -0.0009 0.9993 0.0155>}
 beltRotation_r {parent=pelvis, loc=<2.2368 1.196 7.2613>, rot=<0 0 0.9998 0.0179>}
 base_belt_r {parent=beltRotation_r, loc=<2.7458 -6.1406 -0.8748>, rot=<-0.0628 0 0 -0.998>}
 dyn_belt_r {parent=base_belt_r, loc=<0 0 0>, rot=<0 0 0 1>}
 base_beltBullets {parent=beltRotation_r, loc=<-0.5398 -3.7867 4.7964>, rot=<0.4059 0 0 -0.9139>}
 dyn_beltBullets {parent=base_beltBullets, loc=<0 0 0>, rot=<0 0 0 1>}
 dyn_belt01 {parent=pelvis, loc=<3.5821 0.7293 8.9084>, rot=<-0.5949 0.1297 0.7907 -0.0639>}
 dyn_belt02 {parent=dyn_belt01, loc=<8.4 0 0>, rot=<0 0 0 1>}
 dyn_belt03 {parent=dyn_belt02, loc=<8.4 0 0>, rot=<0 0 0 1>}
 base_holster {parent=dyn_belt03, loc=<9.3384 0.0164 0.0253>, rot=<0.7969 0.048 0.6003 -0.0481>}
 dyn_holster {parent=base_holster, loc=<0 0 0>, rot=<0 0 0 1>}
 base_belt_l {parent=dyn_belt03, loc=<0.0263 -10.6786 2.2345>, rot=<0.119 0.5972 0.0781 -0.7894>}
 dyn_belt_l {parent=base_belt_l, loc=<0 0 0>, rot=<0 0 0 1>}
 spine_05_null {parent=pelvis, type=Null, loc=<32.972 -4.2232 0>, rot=<0 0 0.058 -0.9983>}
 thigh_l_ctrl {controlType=EULER_TRANSFORM, type=Control, loc=<0 0 0>, rot=<0 0 0 1>}
 calf_l_ctrl {controlType=EULER_TRANSFORM, parent=thigh_l_ctrl, type=Control, loc=<0 0 0>, rot=<0 0 0 1>}
 foot_l_ctrl {controlType=EULER_TRANSFORM, parent=calf_l_ctrl, type=Control, loc=<0 0 0>, rot=<0 0 0 1>}
 ball_l_ctrl {controlType=EULER_TRANSFORM, parent=foot_l_ctrl, type=Control, loc=<0 0 0>, rot=<0 0 0 1>}
 thigh_r_ctrl {controlType=EULER_TRANSFORM, type=Control, loc=<0 0 0>, rot=<0 0 0 1>}
 calf_r_ctrl {controlType=EULER_TRANSFORM, parent=thigh_r_ctrl, type=Control, loc=<0 0 0>, rot=<0 0 0 1>}
 foot_r_ctrl {controlType=EULER_TRANSFORM, parent=calf_r_ctrl, type=Control, loc=<0 0 0>, rot=<0 0 0 1>}
 ball_r_ctrl {controlType=EULER_TRANSFORM, parent=foot_r_ctrl, type=Control, loc=<0 0 0>, rot=<0 0 0 1>}
]
variables [
 isCrouching_CR {type=bool, default=False}
 isFalling_CR {type=double, default=0}
 isKneeSliding_CR {type=bool, default=False}
 isSkydiving_CR {type=bool, default=False}
]
