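# Physics asset for the LivelySugar_Shore medium female player body cosmetic, set up for a dangling handkerchief.
# Top and bottom three-bone capsule chains hang from small base bodies through limited-swing constraints, with twist locked on every joint except the first bottom one.
# Unconstrained capsules on the pelvis, right thigh and right forearm, plus a huge root capsule below the origin, appear to serve as colliders for the chains.
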PhysicsAsset /BRCosmetics/Characters/Player/Female/Medium/Bodies/F_MED_LivelySugar_Shore/Meshes/F_MED_LivelySugar_Shore_Physics
bodies {
 base_handkerchiefBottom [{shapeType=Capsule, length=0.505, radius=0.505, shapeName=base_handkerchiefBottom_capsule, rotation=<rot 0 -90 0>}]
 base_handkerchiefTop [{shapeType=Capsule, length=0.505, radius=0.505, shapeName=base_handkerchiefTop_capsule, rotation=<rot 0 -90 0>}]
 dyn_handkerchiefBottom_01 [{
   shapeType=Capsule
   length=3.0638
   radius=2.1387
   shapeName=dyn_handkerchiefTop_01_capsule
   center=<4.1391 -1.0188 -0.2072>
   rotation=<rot -0.4773 85.7885 -79.4678>
  }]
 dyn_handkerchiefBottom_02 [
  {
   shapeType=Capsule
   length=3.0638
   radius=3.0438
   shapeName=dyn_handkerchiefTop_01_capsule
   center=<4.1846 -1.8459 -0.2685>
   rotation=<rot -1.8051 94.1874 -85.5946>
  }
  {
   shapeType=Capsule
   length=2.0139
   radius=2.1459
   shapeName=dyn_handkerchiefBottom_02_capsule
   center=<6.262 -0.9617 1.4818>
   rotation=<rot -80.0002 -0.0002 0.0001>
  }
 ]
 dyn_handkerchiefBottom_03 [{
   shapeType=Capsule
   length=5.1719
   radius=3.0438
   shapeName=dyn_handkerchiefTop_01_capsule
   center=<4.1345 -1.3357 0.0233>
   rotation=<rot -1.6983 92.8079 -85.5521>
  }]
 dyn_handkerchiefTop_01 [{
   shapeType=Capsule
   length=3.0638
   radius=1.9958
   shapeName=dyn_handkerchiefTop_01_capsule
   center=<3.3836 -0.0387 -0.147>
   rotation=<rot 0 89.1194 -95.0617>
  }]
 dyn_handkerchiefTop_02 [{
   shapeType=Capsule
   length=3.0638
   radius=2.3609
   shapeName=dyn_handkerchiefTop_01_capsule
   center=<3.2926 0.1808 -0.5367>
   rotation=<rot 1.1859 102.636 -97.2051>
  }]
 dyn_handkerchiefTop_03 [{
   shapeType=Capsule
   length=3.0638
   radius=2.3609
   shapeName=dyn_handkerchiefTop_01_capsule
   center=<3.5241 1.0375 -0.4944>
   rotation=<rot 0.9305 99.6863 -94.9758>
  }]
 lowerarm_r [{
   shapeType=Capsule
   length=10.568
   radius=8.954
   shapeName=lowerarm_r_capsule
   center=<-8.218 -0.0846 -3.6828>
   rotation=<rot -0.0001 -92.7679 -82.4916>
  }]
 pelvis [{
   shapeType=Capsule
   length=11.0016
   radius=10.8996
   shapeName=pelvis_capsule
   center=<-2.5636 0.5167 -2.0063>
   rotation=<rot 0.3342 -90 1.9677>
  }]
 root [{
   shapeType=Capsule
   length=98.323
   radius=132.0385
   shapeName=root_capsule
   center=<-15.0056 -33.0936 -131.8703>
   rotation=<rot 85.7666 90.016 -180>
  }]
 thigh_r [{
   shapeType=Capsule
   length=24.9331
   radius=10.2002
   shapeName=thigh_r_capsule
   center=<17.3894 1.1723 -2.3936>
   rotation=<rot -0.0001 91.3656 -86.2929>
  }]
}
constraints [
 {bone1Name=dyn_handkerchiefBottom_01, bone2Name=base_handkerchiefBottom, swing1LimitDegrees=70, swing2LimitDegrees=20, twistLimitDegrees=45}
 {
  bone1Name=dyn_handkerchiefTop_01
  bone2Name=base_handkerchiefTop
  swing1LimitDegrees=45
  swing2LimitDegrees=25
  twistLimitDegrees=45
  twistMotion=Locked
 }
 {
  bone1Name=dyn_handkerchiefTop_02
  bone2Name=dyn_handkerchiefTop_01
  swing1LimitDegrees=20
  swing2LimitDegrees=15
  twistLimitDegrees=45
  twistMotion=Locked
 }
 {
  bone1Name=dyn_handkerchiefTop_03
  bone2Name=dyn_handkerchiefTop_02
  swing1LimitDegrees=20
  swing2LimitDegrees=15
  twistLimitDegrees=45
  twistMotion=Locked
 }
 {
  bone1Name=dyn_handkerchiefBottom_02
  bone2Name=dyn_handkerchiefBottom_01
  swing1LimitDegrees=45
  swing2LimitDegrees=20
  twistLimitDegrees=45
  twistMotion=Locked
 }
 {
  bone1Name=dyn_handkerchiefBottom_03
  bone2Name=dyn_handkerchiefBottom_02
  swing1LimitDegrees=45
  swing2LimitDegrees=20
  twistLimitDegrees=45
  twistMotion=Locked
 }
]
