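# Animation Blueprint for the LivelySugar Shore female player body, adding secondary motion to coat, fur trim and chest.
# It copies the attached parent's pose into a Control Rig fed movement states, then weighted constraints place coat and sleeve fur base bones.
# From that cached pose, chained AnimDynamics nodes simulate coat and fur bones.
# Output uses bPbBlendEnabled to toggle an extra chest pass of additive offsets plus angular and linear spring dynamics.
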AnimBlueprint /BRCosmetics/Characters/Player/Female/Medium/Bodies/F_MED_LivelySugar_Shore/Meshes/F_MED_LivelySugar_Shore_AnimBP
target_skeleton=&/F_MED_LivelySugar_Shore_Skeleton
graphs [
 AnimGraph {nodes [
   n1 = "Transform (Modify) Bone - Bone: dyn_chest_l"(ComponentPose=Local To Component(LocalPose=Use cached pose 'allDynCache'()), Alpha=1, Translation="0,0,1", Rotation="0,0,0", Scale="1,1,1") {translationMode=BMM_Additive, translationSpace=BCS_BoneSpace, boneToModify {boneName=dyn_chest_l}, translation {z=1}}
   n2 = "Transform (Modify) Bone - Bone: dyn_chest_r"(ComponentPose=@n1, Alpha=1, Translation="0,0,1", Rotation="0,0,0", Scale="1,1,1") {translationMode=BMM_Additive, translationSpace=BCS_BoneSpace, boneToModify {boneName=dyn_chest_r}, translation {z=1}}
   n3 = Get Sim Space Settings() -> (Sim Space Settings:Anim Phys Sim Space Settings Structure)
   n4 = Anim Dynamics - Bone: dyn_chest_r!(ComponentPose=@n2, Alpha=1, AngularSpringConstant:float=5000, SimSpaceSettings:Anim Phys Sim Space Settings Structure=@n3) -> (Pose) {
    angularBiasOverride=1
    angularDampingOverride=0.9999
    angularSpringConstant=5000
    bAngularSpring=true
    bGravityOverrideInSimSpace=true
    bOverrideAngularBias=true
    bOverrideAngularDamping=true
    bUseGravityOverride=true
    linearDampingOverride=0.9999
    linearSpringConstant=100
    simulationSpace=RootRelative
    boundBone {boneName=dyn_chest_r}
    componentAppliedLinearAccClamp=<2 2 2>
    gravityOverride {z=-980}
    physicsBodyDefinitions [{
      boundBone {boneName=dyn_chest_r}
      boxExtents=<21 21 21>
      constraintSetup {angularLimitsMax=<0 1 5>, angularLimitsMin=<0 -5 -5>, angularTarget=<1 0 0>}
      localJointOffset=<2 0 0>
     }]
    relativeSpaceBone {boneName=spine_04}
    simSpaceSettings {maxAngularAcceleration=45, maxAngularVelocity=8, simSpaceAngularAlpha=0.25}
   }
   n5 = Anim Dynamics - Bone: dyn_chest_l!(ComponentPose=@n4, Alpha=1, AngularSpringConstant:float=5000, SimSpaceSettings:Anim Phys Sim Space Settings Structure=@n3) -> (Pose) {
    angularBiasOverride=1
    angularDampingOverride=0.9999
    angularSpringConstant=5000
    bAngularSpring=true
    bGravityOverrideInSimSpace=true
    bOverrideAngularBias=true
    bOverrideAngularDamping=true
    bUseGravityOverride=true
    linearDampingOverride=0.9999
    linearSpringConstant=100
    simulationSpace=RootRelative
    boundBone {boneName=dyn_chest_l}
    componentAppliedLinearAccClamp=<2 2 2>
    gravityOverride {z=-980}
    physicsBodyDefinitions [{
      boundBone {boneName=dyn_chest_l}
      boxExtents=<20 20 20>
      constraintSetup {angularLimitsMax=<0 1 5>, angularLimitsMin=<0 -5 -5>, angularTarget=<1 0 0>}
      localJointOffset=<2 0 0>
     }]
    relativeSpaceBone {boneName=spine_04}
    simSpaceSettings {maxAngularAcceleration=45, maxAngularVelocity=8, simSpaceAngularAlpha=0.25}
   }
   n6 = Anim Dynamics - Bone: dyn_chest_r!(ComponentPose=@n5, Alpha=1, GravityOverride:Vector="0,0,-200", LinearSpringConstant:float=180) -> (Pose) {
    angularBiasOverride=1
    angularDampingOverride=0.99
    angularSpringConstant=10000
    bGravityOverrideInSimSpace=true
    bLinearSpring=true
    bOverrideAngularBias=true
    bOverrideLinearDamping=true
    bUseGravityOverride=true
    linearDampingOverride=1
    linearSpringConstant=180
    simulationSpace=RootRelative
    boundBone {boneName=dyn_chest_r}
    componentAppliedLinearAccClamp=<5 5 5>
    gravityOverride {z=-200}
    physicsBodyDefinitions [{
      boundBone {boneName=dyn_chest_r}
      boxExtents=<80 80 80>
      constraintSetup {angularLimitsMax=<0 0 0>, angularLimitsMin=<0 0 0>, angularTarget=<1 0 0>, linearAxesMax=<0 0.5 1>, linearAxesMin=<0 -0.5 -0.5>}
      localJointOffset=<0.2 0 0>
     }]
    relativeSpaceBone {boneName=spine_04}
    simSpaceSettings {maxAngularAcceleration=50, maxAngularVelocity=10, simSpaceAngularAlpha=0.15}
   }
   n7 = Anim Dynamics - Bone: dyn_chest_l!(ComponentPose=@n6, Alpha=1, GravityOverride:Vector="0,0,-200", LinearSpringConstant:float=180) -> (Pose) {
    angularBiasOverride=1
    angularDampingOverride=0.99
    angularSpringConstant=10000
    bGravityOverrideInSimSpace=true
    bLinearSpring=true
    bOverrideAngularBias=true
    bOverrideLinearDamping=true
    bUseGravityOverride=true
    linearDampingOverride=1
    linearSpringConstant=180
    simulationSpace=RootRelative
    boundBone {boneName=dyn_chest_l}
    componentAppliedLinearAccClamp=<5 5 5>
    gravityOverride {z=-200}
    physicsBodyDefinitions [{
      boundBone {boneName=dyn_chest_l}
      boxExtents=<90 90 90>
      constraintSetup {angularLimitsMax=<0 0 0>, angularLimitsMin=<0 0 0>, angularTarget=<1 0 0>, linearAxesMax=<0 0.5 1>, linearAxesMin=<0 -0.5 -0.5>}
      localJointOffset=<0.2 0 0>
     }]
    relativeSpaceBone {boneName=spine_04}
    simSpaceSettings {maxAngularAcceleration=50, maxAngularVelocity=10, simSpaceAngularAlpha=0.15}
   }
   Output Pose(Result=Blend Poses by bool(
     BlendPose_0=Component To Local(ComponentPose=@n7)
     BlendPose_1=Use cached pose 'allDynCache'()
     bActiveValue=Get bPbBlendEnabled() -> (bPbBlendEnabled:bool)
    )) {layerGroup=None}
   n8 = Control Rig!(
    Source=Copy Pose From Mesh() {bUseAttachedParent=true}
    Alpha
    isCrouching_CR:bool=Get bIsCrouching()
    isKneeSliding_CR:bool=Get bIsKneeSliding()
    isSkydiving_CR:bool=Get bIsSkyDiving()
    isFalling_CR:double=Get isFallingNotMoving() -> (isFallingNotMoving:double)
   ) -> (Pose) {controlRigAssetReference {blueprintRigClass=&/F_MED_LivelySugar_Shore_CtrlRig.F_MED_LivelySugar_Shore_CtrlRig_C}}
   n9 = "Constraint - base_coatFur_01_r (PPPPPPPP)"!(
    ComponentPose=Local To Component(LocalPose=@n8)
    Alpha=1
    ConstraintWeights_0=0.2327
    ConstraintWeights_1:float=0.2235
    ConstraintWeights_2:float=0.2006
    ConstraintWeights_3:float=0.1223
    ConstraintWeights_4:float=0.1008
    ConstraintWeights_5:float=0.0697
    ConstraintWeights_6:float=0.0462
    ConstraintWeights_7:float=0.0042
   ) -> (Pose) {
    boneToModify {boneName=base_coatFur_01_r}
    constraintSetup [
     {targetBone {boneName=dyn_coatVolume01_r}}
     {targetBone {boneName=deform_pec_r}}
     {targetBone {boneName=spine_04}}
     {targetBone {boneName=spine_05}}
     {targetBone {boneName=clavicle_r}}
     {targetBone {boneName=upperarm_twist_01_r}}
     {targetBone {boneName=deform_upperarm_fr_r}}
     {targetBone {boneName=deform_upperarm_up_r}}
    ]
    constraintWeights [1,1,1,1,1,1,1,1]
   }
   n10 = "Constraint - base_coatFur_02_r (PPPPPPPP)"!(
    ComponentPose=@n9
    Alpha=1
    ConstraintWeights_0=0.3459
    ConstraintWeights_1:float=0.2579
    ConstraintWeights_2:float=0.1523
    ConstraintWeights_3:float=0.0907
    ConstraintWeights_4:float=0.0613
    ConstraintWeights_5:float=0.0505
    ConstraintWeights_6:float=0.0309
    ConstraintWeights_7:float=0.0104
   ) -> (Pose) {
    boneToModify {boneName=base_coatFur_02_r}
    constraintSetup [
     {targetBone {boneName=clavicle_r}}
     {targetBone {boneName=upperarm_twist_01_r}}
     {targetBone {boneName=deform_upperarm_fr_r}}
     {targetBone {boneName=deform_upperarm_up_r}}
     {targetBone {boneName=spine_05}}
     {targetBone {boneName=dyn_coatVolume01_r}}
     {targetBone {boneName=spine_04}}
     {targetBone {boneName=deform_pec_r}}
    ]
    constraintWeights [1,1,1,1,1,1,1,1]
   }
   n11 = "Constraint - base_coatFur_03_r (PPPPPPP)"!(
    ComponentPose=@n10
    Alpha=1
    ConstraintWeights_0=0.5107
    ConstraintWeights_1:float=0.2129
    ConstraintWeights_2:float=0.164
    ConstraintWeights_3:float=0.0532
    ConstraintWeights_4:float=0.0338
    ConstraintWeights_5:float=0.0249
    ConstraintWeights_6:float=0.0006
   ) -> (Pose) {
    boneToModify {boneName=base_coatFur_03_r}
    constraintSetup [
     {targetBone {boneName=clavicle_r}}
     {targetBone {boneName=upperarm_twist_01_r}}
     {targetBone {boneName=deform_upperarm_up_r}}
     {targetBone {boneName=deform_upperarm_bk_r}}
     {targetBone {boneName=deform_upperarm_fr_r}}
     {targetBone {boneName=spine_05}}
     {targetBone {boneName=spine_04}}
    ]
    constraintWeights [1,1,1,1,1,1,1]
   }
   n12 = "Constraint - base_coatFur_04_r (PPPPPP)"!(
    ComponentPose=@n11
    Alpha=1
    ConstraintWeights_0=0.5344
    ConstraintWeights_1:float=0.4504
    ConstraintWeights_2:float=0.0068
    ConstraintWeights_3:float=0.0039
    ConstraintWeights_4:float=0.0039
    ConstraintWeights_5:float=0.0006
   ) -> (Pose) {
    boneToModify {boneName=base_coatFur_04_r}
    constraintSetup [
     {targetBone {boneName=clavicle_r}}
     {targetBone {boneName=spine_05}}
     {targetBone {boneName=clavicle_l}}
     {targetBone {boneName=deform_upperarm_up_r}}
     {targetBone {boneName=upperarm_twist_01_r}}
     {targetBone {boneName=deform_upperarm_bk_r}}
    ]
    constraintWeights [1,1,1,1,1,1]
   }
   n13 = "Constraint - base_coatFur_c (PPP)"!(ComponentPose=@n12, Alpha=1, ConstraintWeights_0=0.8396, ConstraintWeights_1:float=0.0964, ConstraintWeights_2:float=0.064) -> (Pose) {
    boneToModify {boneName=base_coatFur_c}
    constraintSetup [{targetBone {boneName=spine_05}}, {targetBone {boneName=clavicle_r}}, {targetBone {boneName=clavicle_l}}]
    constraintWeights [1,1,1]
   }
   n14 = "Constraint - base_coatFur_04_l (PPPPPP)"!(
    ComponentPose=@n13
    Alpha=1
    ConstraintWeights_0=0.4932
    ConstraintWeights_1:float=0.4893
    ConstraintWeights_2:float=0.0096
    ConstraintWeights_3:float=0.0034
    ConstraintWeights_4:float=0.0034
    ConstraintWeights_5:float=0.0011
   ) -> (Pose) {
    boneToModify {boneName=base_coatFur_04_l}
    constraintSetup [
     {targetBone {boneName=clavicle_l}}
     {targetBone {boneName=spine_05}}
     {targetBone {boneName=clavicle_r}}
     {targetBone {boneName=upperarm_twist_01_l}}
     {targetBone {boneName=deform_upperarm_up_l}}
     {targetBone {boneName=deform_upperarm_bk_l}}
    ]
    constraintWeights [1,1,1,1,1,1]
   }
   n15 = "Constraint - base_coatFur_03_l (PPPPPP)"!(
    ComponentPose=@n14
    Alpha=1
    ConstraintWeights_0=0.5662
    ConstraintWeights_1:float=0.1761
    ConstraintWeights_2:float=0.1482
    ConstraintWeights_3:float=0.0566
    ConstraintWeights_4:float=0.0319
    ConstraintWeights_5:float=0.0209
   ) -> (Pose) {
    boneToModify {boneName=base_coatFur_03_l}
    constraintSetup [
     {targetBone {boneName=clavicle_l}}
     {targetBone {boneName=upperarm_twist_01_l}}
     {targetBone {boneName=deform_upperarm_up_l}}
     {targetBone {boneName=deform_upperarm_bk_l}}
     {targetBone {boneName=spine_05}}
     {targetBone {boneName=deform_upperarm_fr_l}}
    ]
    constraintWeights [1,1,1,1,1,1]
   }
   n16 = "Constraint - base_coatFur_02_l (PPPPPPPP)"!(
    ComponentPose=@n15
    Alpha=1
    ConstraintWeights_0=0.368
    ConstraintWeights_1:float=0.2424
    ConstraintWeights_2:float=0.1453
    ConstraintWeights_3:float=0.1042
    ConstraintWeights_4:float=0.0547
    ConstraintWeights_5:float=0.0538
    ConstraintWeights_6:float=0.0252
    ConstraintWeights_7:float=0.0064
   ) -> (Pose) {
    boneToModify {boneName=base_coatFur_02_l}
    constraintSetup [
     {targetBone {boneName=clavicle_l}}
     {targetBone {boneName=upperarm_twist_01_l}}
     {targetBone {boneName=deform_upperarm_fr_l}}
     {targetBone {boneName=deform_upperarm_up_l}}
     {targetBone {boneName=dyn_coatVolume01_l}}
     {targetBone {boneName=spine_05}}
     {targetBone {boneName=spine_04}}
     {targetBone {boneName=deform_pec_l}}
    ]
    constraintWeights [1,1,1,1,1,1,1,1]
   }
   n17 = "Constraint - base_coatFur_01_l (PPPPPPPP)"!(
    ComponentPose=@n16
    Alpha=1
    ConstraintWeights_0=0.2334
    ConstraintWeights_1:float=0.229
    ConstraintWeights_2:float=0.2141
    ConstraintWeights_3:float=0.1191
    ConstraintWeights_4:float=0.1
    ConstraintWeights_5:float=0.0583
    ConstraintWeights_6:float=0.0427
    ConstraintWeights_7:float=0.0034
   ) -> (Pose) {
    boneToModify {boneName=base_coatFur_01_l}
    constraintSetup [
     {targetBone {boneName=deform_pec_l}}
     {targetBone {boneName=dyn_coatVolume01_l}}
     {targetBone {boneName=spine_04}}
     {targetBone {boneName=spine_05}}
     {targetBone {boneName=clavicle_l}}
     {targetBone {boneName=upperarm_twist_01_l}}
     {targetBone {boneName=deform_upperarm_fr_l}}
     {targetBone {boneName=deform_upperarm_up_l}}
    ]
    constraintWeights [1,1,1,1,1,1,1,1]
   }
   n18 = "Constraint - base_inSleeveFur_l (PPPPP)"!(
    ComponentPose=@n17
    Alpha=1
    ConstraintWeights_0=0.451
    ConstraintWeights_1:float=0.323
    ConstraintWeights_2:float=0.2253
    ConstraintWeights_3:float=0.0004
    ConstraintWeights_4:float=0.0003
   ) -> (Pose) {
    boneToModify {boneName=base_inSleeveFur_l}
    constraintSetup [
     {targetBone {boneName=lowerarm_twist_02_l}}
     {targetBone {boneName=lowerarm_twist_01_l}}
     {targetBone {boneName=lowerarm_l}}
     {targetBone {boneName=deform_elbow_in_l}}
     {targetBone {boneName=upperarm_l}}
    ]
    constraintWeights [1,1,1,1,1]
   }
   n19 = "Constraint - base_bottomSleeveFur_l (PPP)"!(ComponentPose=@n18, Alpha=1, ConstraintWeights_0=0.6217, ConstraintWeights_1:float=0.1979, ConstraintWeights_2:float=0.1804) -> (Pose) {
    boneToModify {boneName=base_bottomSleeveFur_l}
    constraintSetup [{targetBone {boneName=lowerarm_twist_02_l}}, {targetBone {boneName=lowerarm_l}}, {targetBone {boneName=lowerarm_twist_01_l}}]
    constraintWeights [1,1,1]
   }
   n20 = "Constraint - base_outSleeveFur_l (PPPPP)"!(
    ComponentPose=@n19
    Alpha=1
    ConstraintWeights_0=0.5649
    ConstraintWeights_1:float=0.2496
    ConstraintWeights_2:float=0.1755
    ConstraintWeights_3:float=0.0065
    ConstraintWeights_4:float=0.0035
   ) -> (Pose) {
    boneToModify {boneName=base_outSleeveFur_l}
    constraintSetup [
     {targetBone {boneName=lowerarm_twist_02_l}}
     {targetBone {boneName=lowerarm_twist_01_l}}
     {targetBone {boneName=lowerarm_l}}
     {targetBone {boneName=upperarm_l}}
     {targetBone {boneName=deform_elbow_out_l}}
    ]
    constraintWeights [1,1,1,1,1]
   }
   n21 = "Constraint - base_topSleeveFur_l (PPPPP)"!(
    ComponentPose=@n20
    Alpha=1
    ConstraintWeights_0=0.5768
    ConstraintWeights_1:float=0.215
    ConstraintWeights_2:float=0.2005
    ConstraintWeights_3:float=0.0072
    ConstraintWeights_4:float=0.0006
   ) -> (Pose) {
    boneToModify {boneName=base_topSleeveFur_l}
    constraintSetup [
     {targetBone {boneName=lowerarm_twist_02_l}}
     {targetBone {boneName=lowerarm_twist_01_l}}
     {targetBone {boneName=lowerarm_l}}
     {targetBone {boneName=upperarm_l}}
     {targetBone {boneName=deform_elbow_out_l}}
    ]
    constraintWeights [1,1,1,1,1]
   }
   n22 = "Constraint - base_topSleeveFur_r (PPPPP)"!(
    ComponentPose=@n21
    Alpha=1
    ConstraintWeights_0=0.5768
    ConstraintWeights_1:float=0.215
    ConstraintWeights_2:float=0.2005
    ConstraintWeights_3:float=0.0072
    ConstraintWeights_4:float=0.0006
   ) -> (Pose) {
    boneToModify {boneName=base_topSleeveFur_r}
    constraintSetup [
     {targetBone {boneName=lowerarm_twist_02_r}}
     {targetBone {boneName=lowerarm_twist_01_r}}
     {targetBone {boneName=lowerarm_r}}
     {targetBone {boneName=upperarm_r}}
     {targetBone {boneName=deform_elbow_out_r}}
    ]
    constraintWeights [1,1,1,1,1]
   }
   n23 = "Constraint - base_inSleeveFur_r (PPP)"!(ComponentPose=@n22, Alpha=1, ConstraintWeights_0=0.4532, ConstraintWeights_1:float=0.3175, ConstraintWeights_2:float=0.2293) -> (Pose) {
    boneToModify {boneName=base_inSleeveFur_r}
    constraintSetup [{targetBone {boneName=lowerarm_twist_02_r}}, {targetBone {boneName=lowerarm_twist_01_r}}, {targetBone {boneName=lowerarm_r}}]
    constraintWeights [1,1,1]
   }
   n24 = "Constraint - base_outSleeveFur_r (PPPPP)"!(
    ComponentPose=@n23
    Alpha=1
    ConstraintWeights_0=0.5649
    ConstraintWeights_1:float=0.2496
    ConstraintWeights_2:float=0.1755
    ConstraintWeights_3:float=0.0065
    ConstraintWeights_4:float=0.0035
   ) -> (Pose) {
    boneToModify {boneName=base_outSleeveFur_r}
    constraintSetup [
     {targetBone {boneName=lowerarm_twist_02_r}}
     {targetBone {boneName=lowerarm_twist_01_r}}
     {targetBone {boneName=lowerarm_r}}
     {targetBone {boneName=upperarm_r}}
     {targetBone {boneName=deform_elbow_out_r}}
    ]
    constraintWeights [1,1,1,1,1]
   }
   n25 = "Constraint - base_bottomSleeveFur_r (PPP)"!(ComponentPose=@n24, Alpha=1, ConstraintWeights_0=0.6215, ConstraintWeights_1:float=0.1982, ConstraintWeights_2:float=0.1803) -> (Pose) {
    boneToModify {boneName=base_bottomSleeveFur_r}
    constraintSetup [{targetBone {boneName=lowerarm_twist_02_r}}, {targetBone {boneName=lowerarm_r}}, {targetBone {boneName=lowerarm_twist_01_r}}]
    constraintWeights [1,1,1]
   }
   Save cached pose 'preSimCache'(Pose=Component To Local(ComponentPose=@n25)) {cacheName=preSimCache}
   n26 = Anim Dynamics - Bone: dyn_coatFront_l!(
    ComponentPose=Local To Component(LocalPose=Use cached pose 'preSimCache'())
    Alpha=1
    GravityOverride:Vector="0,0,-980"
    ExternalForce="0,0,0"
   ) -> (Pose) {
    angularDampingOverride=1
    angularSpringConstant=8000
    bAngularSpring=true
    bOverrideAngularDamping=true
    bUseGravityOverride=true
    gravityScale=0.35
    simulationSpace=BoneRelative
    alphaScaleBiasClamp {bInterpResult=true, interpSpeedDecreasing=5}
    boundBone {boneName=dyn_coatFront_l}
    componentAppliedLinearAccClamp=<4000 4000 4000>
    componentLinearAccScale=<0.4 0.4 0.4>
    componentLinearVelScale=<0.4 0.4 0.4>
    gravityOverride {z=-980}
    physicsBodyDefinitions [{
      boundBone {boneName=dyn_coatFront_l}
      boxExtents=<26 26 26>
      constraintSetup {angularLimitsMax=<0 15 15>, angularLimitsMin=<0 -15 -60>, angularTarget=<1 0 0>}
      localJointOffset=<5 0 0>
     }]
    relativeSpaceBone {boneName=root}
    simSpaceSettings {simSpaceAngularAlpha=0.2}
   }
   n27 = Anim Dynamics - Bone: dyn_coatSide_l!(ComponentPose=@n26, Alpha=1, GravityOverride:Vector="0,0,-980", ExternalForce="0,0,0") -> (Pose) {
    angularDampingOverride=1
    angularSpringConstant=8000
    bAngularSpring=true
    bOverrideAngularDamping=true
    bUseGravityOverride=true
    gravityScale=0.35
    simulationSpace=BoneRelative
    alphaScaleBiasClamp {bInterpResult=true, interpSpeedDecreasing=5}
    boundBone {boneName=dyn_coatSide_l}
    componentAppliedLinearAccClamp=<4000 4000 4000>
    componentLinearAccScale=<0.4 0.4 0.4>
    componentLinearVelScale=<0.4 0.4 0.4>
    gravityOverride {z=-980}
    physicsBodyDefinitions [{
      boundBone {boneName=dyn_coatSide_l}
      boxExtents=<26 26 26>
      constraintSetup {angularLimitsMax=<0 15 15>, angularLimitsMin=<0 -15 -60>, angularTarget=<1 0 0>}
      localJointOffset=<5 0 0>
     }]
    relativeSpaceBone {boneName=root}
    simSpaceSettings {simSpaceAngularAlpha=0.2}
   }
   n28 = Anim Dynamics - Bone: dyn_coatSideBack_l!(ComponentPose=@n27, Alpha=1, GravityOverride:Vector="0,0,-980", ExternalForce="0,0,0") -> (Pose) {
    angularDampingOverride=1
    angularSpringConstant=8000
    bAngularSpring=true
    bOverrideAngularDamping=true
    bUseGravityOverride=true
    gravityScale=0.35
    simulationSpace=BoneRelative
    alphaScaleBiasClamp {bInterpResult=true, interpSpeedDecreasing=5}
    boundBone {boneName=dyn_coatSideBack_l}
    componentAppliedLinearAccClamp=<4000 4000 4000>
    componentLinearAccScale=<0.4 0.4 0.4>
    componentLinearVelScale=<0.4 0.4 0.4>
    gravityOverride {z=-980}
    physicsBodyDefinitions [{
      boundBone {boneName=dyn_coatSideBack_l}
      boxExtents=<26 26 26>
      constraintSetup {angularLimitsMax=<0 15 15>, angularLimitsMin=<0 -15 -60>, angularTarget=<1 0 0>}
      localJointOffset=<5 0 0>
     }]
    relativeSpaceBone {boneName=root}
    simSpaceSettings {simSpaceAngularAlpha=0.2}
   }
   n29 = Anim Dynamics - Bone: dyn_coatBack!(ComponentPose=@n28, Alpha=1, GravityOverride:Vector="0,0,-980", ExternalForce="0,0,0") -> (Pose) {
    angularDampingOverride=1
    angularSpringConstant=8000
    bAngularSpring=true
    bOverrideAngularDamping=true
    bUseGravityOverride=true
    gravityScale=0.35
    simulationSpace=BoneRelative
    alphaScaleBiasClamp {bInterpResult=true, interpSpeedDecreasing=5}
    boundBone {boneName=dyn_coatBack}
    componentAppliedLinearAccClamp=<4000 4000 4000>
    componentLinearAccScale=<0.4 0.4 0.4>
    componentLinearVelScale=<0.4 0.4 0.4>
    gravityOverride {z=-980}
    physicsBodyDefinitions [{
      boundBone {boneName=dyn_coatBack}
      boxExtents=<26 26 26>
      constraintSetup {angularLimitsMax=<0 15 15>, angularLimitsMin=<0 -15 -60>, angularTarget=<1 0 0>}
      localJointOffset=<5 0 0>
     }]
    relativeSpaceBone {boneName=root}
    simSpaceSettings {simSpaceAngularAlpha=0.2}
   }
   n30 = Anim Dynamics - Bone: dyn_coatFront_r!(ComponentPose=@n29, Alpha=1, GravityOverride:Vector="0,0,-980", ExternalForce="0,0,0") -> (Pose) {
    angularDampingOverride=1
    angularSpringConstant=8000
    bAngularSpring=true
    bOverrideAngularDamping=true
    bUseGravityOverride=true
    gravityScale=0.35
    simulationSpace=BoneRelative
    alphaScaleBiasClamp {bInterpResult=true, interpSpeedDecreasing=5}
    boundBone {boneName=dyn_coatFront_r}
    componentAppliedLinearAccClamp=<4000 4000 4000>
    componentLinearAccScale=<0.4 0.4 0.4>
    componentLinearVelScale=<0.4 0.4 0.4>
    gravityOverride {z=-980}
    physicsBodyDefinitions [{
      boundBone {boneName=dyn_coatFront_r}
      boxExtents=<26 26 26>
      constraintSetup {angularLimitsMax=<0 15 15>, angularLimitsMin=<0 -15 -60>, angularTarget=<1 0 0>}
      localJointOffset=<5 0 0>
     }]
    relativeSpaceBone {boneName=root}
    simSpaceSettings {simSpaceAngularAlpha=0.2}
   }
   n31 = Anim Dynamics - Bone: dyn_coatSide_r!(ComponentPose=@n30, Alpha=1, GravityOverride:Vector="0,0,-980", ExternalForce="0,0,0") -> (Pose) {
    angularDampingOverride=1
    angularSpringConstant=8000
    bAngularSpring=true
    bOverrideAngularDamping=true
    bUseGravityOverride=true
    gravityScale=0.35
    simulationSpace=BoneRelative
    alphaScaleBiasClamp {bInterpResult=true, interpSpeedDecreasing=5}
    boundBone {boneName=dyn_coatSide_r}
    componentAppliedLinearAccClamp=<4000 4000 4000>
    componentLinearAccScale=<0.4 0.4 0.4>
    componentLinearVelScale=<0.4 0.4 0.4>
    gravityOverride {z=-980}
    physicsBodyDefinitions [{
      boundBone {boneName=dyn_coatSide_r}
      boxExtents=<26 26 26>
      constraintSetup {angularLimitsMax=<0 15 15>, angularLimitsMin=<0 -15 -60>, angularTarget=<1 0 0>}
      localJointOffset=<5 0 0>
     }]
    relativeSpaceBone {boneName=root}
    simSpaceSettings {simSpaceAngularAlpha=0.2}
   }
   n32 = Anim Dynamics - Bone: dyn_coatSideBack_r!(ComponentPose=@n31, Alpha=1, GravityOverride:Vector="0,0,-980", ExternalForce="0,0,0") -> (Pose) {
    angularDampingOverride=1
    angularSpringConstant=8000
    bAngularSpring=true
    bOverrideAngularDamping=true
    bUseGravityOverride=true
    gravityScale=0.35
    simulationSpace=BoneRelative
    alphaScaleBiasClamp {bInterpResult=true, interpSpeedDecreasing=5}
    boundBone {boneName=dyn_coatSideBack_r}
    componentAppliedLinearAccClamp=<4000 4000 4000>
    componentLinearAccScale=<0.4 0.4 0.4>
    componentLinearVelScale=<0.4 0.4 0.4>
    gravityOverride {z=-980}
    physicsBodyDefinitions [{
      boundBone {boneName=dyn_coatSideBack_r}
      boxExtents=<26 26 26>
      constraintSetup {angularLimitsMax=<0 15 15>, angularLimitsMin=<0 -15 -60>, angularTarget=<1 0 0>}
      localJointOffset=<5 0 0>
     }]
    relativeSpaceBone {boneName=root}
    simSpaceSettings {simSpaceAngularAlpha=0.2}
   }
   n33 = Anim Dynamics - Bone: dyn_coatFur_01_l!(ComponentPose=@n32, Alpha=1, GravityOverride:Vector="0,0,-980") -> (Pose) {
    angularDampingOverride=1
    angularSpringConstant=5500
    bAngularSpring=true
    bOverrideAngularDamping=true
    bUseGravityOverride=true
    gravityScale=0.35
    simulationSpace=RootRelative
    alphaBoolBlend {blendOption=Linear}
    alphaScaleBiasClamp {bInterpResult=true}
    boundBone {boneName=dyn_coatFur_01_l}
    componentAppliedLinearAccClamp=<4000 4000 3000>
    componentLinearAccScale=<0.3 0.3 0.3>
    componentLinearVelScale=<0.4 0.4 0.4>
    gravityOverride {z=-980}
    physicsBodyDefinitions [{
      boundBone {boneName=dyn_coatFur_01_l}
      boxExtents=<16 16 16>
      constraintSetup {angularLimitsMax=<10 20 15>, angularLimitsMin=<-10 -20 -60>, angularTarget=<1 0 0>}
      localJointOffset=<-4 -2 -2>
     }]
    relativeSpaceBone {boneName=pelvis}
    simSpaceSettings {simSpaceAngularAlpha=0.4}
   }
   n34 = Anim Dynamics - Bone: dyn_coatFur_02_l!(ComponentPose=@n33, Alpha=0.6, GravityOverride:Vector="0,0,-980") -> (Pose) {
    angularDampingOverride=1
    angularSpringConstant=5500
    bAngularSpring=true
    bOverrideAngularDamping=true
    bUseGravityOverride=true
    gravityScale=0.35
    simulationSpace=RootRelative
    alphaBoolBlend {blendOption=Linear}
    alphaScaleBiasClamp {bInterpResult=true}
    boundBone {boneName=dyn_coatFur_02_l}
    componentAppliedLinearAccClamp=<4000 4000 3000>
    componentLinearAccScale=<0.3 0.3 0.3>
    componentLinearVelScale=<0.4 0.4 0.4>
    gravityOverride {z=-980}
    physicsBodyDefinitions [{
      boundBone {boneName=dyn_coatFur_02_l}
      boxExtents=<17 17 17>
      constraintSetup {angularLimitsMax=<10 20 15>, angularLimitsMin=<-10 -20 -60>, angularTarget=<1 0 0>}
      localJointOffset=<-3 0 -2>
     }]
    relativeSpaceBone {boneName=pelvis}
    simSpaceSettings {simSpaceAngularAlpha=0.4}
   }
   n35 = Anim Dynamics - Bone: dyn_coatFur_03_l!(ComponentPose=@n34, Alpha=0.7, GravityOverride:Vector="0,0,-980") -> (Pose) {
    angularDampingOverride=1
    angularSpringConstant=7000
    bAngularSpring=true
    bOverrideAngularDamping=true
    bUseGravityOverride=true
    gravityScale=0.35
    simulationSpace=RootRelative
    alphaBoolBlend {blendOption=Linear}
    alphaScaleBiasClamp {bInterpResult=true}
    boundBone {boneName=dyn_coatFur_03_l}
    componentAppliedLinearAccClamp=<4000 4000 3000>
    componentLinearAccScale=<0.2 2 0.3>
    componentLinearVelScale=<0.2 0.3 0.4>
    gravityOverride {z=-980}
    physicsBodyDefinitions [{
      boundBone {boneName=dyn_coatFur_03_l}
      boxExtents=<17 17 17>
      constraintSetup {angularLimitsMax=<10 20 15>, angularLimitsMin=<-10 -20 -60>, angularTarget=<1 0 0>}
      localJointOffset=<-4 0 -3>
     }]
    relativeSpaceBone {boneName=pelvis}
    simSpaceSettings {simSpaceAngularAlpha=0.4}
   }
   n36 = Anim Dynamics - Bone: dyn_coatFur_04_l!(ComponentPose=@n35, Alpha=1, GravityOverride:Vector="0,0,-980") -> (Pose) {
    angularDampingOverride=1
    angularSpringConstant=5500
    bAngularSpring=true
    bOverrideAngularDamping=true
    bUseGravityOverride=true
    gravityScale=0.35
    simulationSpace=RootRelative
    alphaBoolBlend {blendOption=Linear}
    alphaScaleBiasClamp {bInterpResult=true}
    boundBone {boneName=dyn_coatFur_04_l}
    componentAppliedLinearAccClamp=<4000 4000 3000>
    componentLinearAccScale=<0.3 0.3 0.3>
    componentLinearVelScale=<0.4 0.4 0.4>
    gravityOverride {z=-980}
    physicsBodyDefinitions [{
      boundBone {boneName=dyn_coatFur_04_l}
      boxExtents=<19 19 19>
      constraintSetup {angularLimitsMax=<10 20 15>, angularLimitsMin=<-10 -20 -60>, angularTarget=<1 0 0>}
      localJointOffset=<-5 0 0>
     }]
    relativeSpaceBone {boneName=pelvis}
    simSpaceSettings {simSpaceAngularAlpha=0.4}
   }
   n37 = Anim Dynamics - Bone: dyn_coatFur_c!(ComponentPose=@n36, Alpha=1, GravityOverride:Vector="0,0,-980") -> (Pose) {
    angularDampingOverride=1
    angularSpringConstant=5500
    bAngularSpring=true
    bOverrideAngularDamping=true
    bUseGravityOverride=true
    gravityScale=0.35
    simulationSpace=RootRelative
    alphaBoolBlend {blendOption=Linear}
    alphaScaleBiasClamp {bInterpResult=true}
    boundBone {boneName=dyn_coatFur_c}
    componentAppliedLinearAccClamp=<4000 4000 3000>
    componentLinearAccScale=<0.3 0.3 0.3>
    componentLinearVelScale=<0.4 0.4 0.4>
    gravityOverride {z=-980}
    physicsBodyDefinitions [{
      boundBone {boneName=dyn_coatFur_c}
      boxExtents=<16 16 16>
      constraintSetup {angularLimitsMax=<10 20 15>, angularLimitsMin=<-10 -20 -60>, angularTarget=<1 0 0>}
      localJointOffset=<-4 0 0>
     }]
    relativeSpaceBone {boneName=pelvis}
    simSpaceSettings {simSpaceAngularAlpha=0.4}
   }
   n38 = Anim Dynamics - Bone: dyn_coatFur_01_r!(ComponentPose=@n37, Alpha=1, GravityOverride:Vector="0,0,-980") -> (Pose) {
    angularDampingOverride=1
    angularSpringConstant=5500
    bAngularSpring=true
    bOverrideAngularDamping=true
    bUseGravityOverride=true
    gravityScale=0.35
    simulationSpace=RootRelative
    alphaBoolBlend {blendOption=Linear}
    alphaScaleBiasClamp {bInterpResult=true}
    boundBone {boneName=dyn_coatFur_01_r}
    componentAppliedLinearAccClamp=<4000 4000 3000>
    componentLinearAccScale=<0.3 0.3 0.3>
    componentLinearVelScale=<0.4 0.4 0.4>
    gravityOverride {z=-980}
    physicsBodyDefinitions [{
      boundBone {boneName=dyn_coatFur_01_r}
      boxExtents=<10 10 10>
      constraintSetup {angularLimitsMax=<10 30 15>, angularLimitsMin=<-10 -30 -60>, angularTarget=<1 0 0>}
      localJointOffset=<-4 6 0>
     }]
    relativeSpaceBone {boneName=pelvis}
    simSpaceSettings {simSpaceAngularAlpha=0.4}
   }
   n39 = Anim Dynamics - Bone: dyn_coatFur_02_r!(ComponentPose=@n38, Alpha=0.8, GravityOverride:Vector="0,0,-980") -> (Pose) {
    angularDampingOverride=1
    angularSpringConstant=5500
    bAngularSpring=true
    bOverrideAngularDamping=true
    bUseGravityOverride=true
    gravityScale=0.35
    simulationSpace=RootRelative
    alphaBoolBlend {blendOption=Linear}
    alphaScaleBiasClamp {bInterpResult=true}
    boundBone {boneName=dyn_coatFur_02_r}
    componentAppliedLinearAccClamp=<4000 4000 3000>
    componentLinearAccScale=<0.3 0.3 0.3>
    componentLinearVelScale=<0.4 0.4 0.4>
    gravityOverride {z=-980}
    physicsBodyDefinitions [{
      boundBone {boneName=dyn_coatFur_02_r}
      boxExtents=<14 14 14>
      constraintSetup {angularLimitsMax=<10 20 15>, angularLimitsMin=<-10 -20 -60>, angularTarget=<1 0 0>}
      localJointOffset=<-4 0 0>
     }]
    relativeSpaceBone {boneName=pelvis}
    simSpaceSettings {simSpaceAngularAlpha=0.4}
   }
   n40 = Anim Dynamics - Bone: dyn_coatFur_03_r!(ComponentPose=@n39, Alpha=1, GravityOverride:Vector="0,0,-980") -> (Pose) {
    angularDampingOverride=1
    angularSpringConstant=5500
    bAngularSpring=true
    bOverrideAngularDamping=true
    bUseGravityOverride=true
    gravityScale=0.35
    simulationSpace=RootRelative
    alphaBoolBlend {blendOption=Linear}
    alphaScaleBiasClamp {bInterpResult=true}
    boundBone {boneName=dyn_coatFur_03_r}
    componentAppliedLinearAccClamp=<4000 4000 3000>
    componentLinearAccScale=<0.3 0.3 0.3>
    componentLinearVelScale=<0.4 0.4 0.4>
    gravityOverride {z=-980}
    physicsBodyDefinitions [{
      boundBone {boneName=dyn_coatFur_03_r}
      boxExtents=<14 14 14>
      constraintSetup {angularLimitsMax=<10 30 15>, angularLimitsMin=<-10 -30 -60>, angularTarget=<1 0 0>}
      localJointOffset=<-2 0 -3>
     }]
    relativeSpaceBone {boneName=pelvis}
    simSpaceSettings {simSpaceAngularAlpha=0.4}
   }
   n41 = Anim Dynamics - Bone: dyn_coatFur_04_r!(ComponentPose=@n40, Alpha=1, GravityOverride:Vector="0,0,-980") -> (Pose) {
    angularDampingOverride=1
    angularSpringConstant=7000
    bAngularSpring=true
    bOverrideAngularDamping=true
    bUseGravityOverride=true
    gravityScale=0.35
    simulationSpace=RootRelative
    alphaBoolBlend {blendOption=Linear}
    alphaScaleBiasClamp {bInterpResult=true}
    boundBone {boneName=dyn_coatFur_04_r}
    componentAppliedLinearAccClamp=<4000 4000 3000>
    componentLinearAccScale=<0.3 0.3 0.3>
    componentLinearVelScale=<0.4 0.4 0.4>
    gravityOverride {z=-980}
    physicsBodyDefinitions [{
      boundBone {boneName=dyn_coatFur_04_r}
      boxExtents=<14 14 14>
      constraintSetup {angularLimitsMax=<10 20 15>, angularLimitsMin=<-10 -20 -60>, angularTarget=<1 0 0>}
      localJointOffset=<-4 0 2>
     }]
    relativeSpaceBone {boneName=pelvis}
    simSpaceSettings {simSpaceAngularAlpha=0.4}
   }
   n42 = Anim Dynamics - Bone: dyn_outSleeveFur_l!(ComponentPose=@n41, GravityOverride:Vector="0,0,-980") -> (Pose) {
    alpha=0.75
    angularDampingOverride=1
    angularSpringConstant=1200
    bAngularSpring=true
    bOverrideAngularDamping=true
    bUseGravityOverride=true
    gravityScale=0.35
    simulationSpace=RootRelative
    alphaBoolBlend {blendOption=Linear}
    alphaScaleBiasClamp {bInterpResult=true}
    boundBone {boneName=dyn_outSleeveFur_l}
    componentAppliedLinearAccClamp=<3000 3000 3000>
    componentLinearAccScale=<0.3 0.3 0.3>
    componentLinearVelScale=<0.3 0.3 0.3>
    gravityOverride {z=-980}
    physicsBodyDefinitions [{
      boundBone {boneName=dyn_outSleeveFur_l}
      boxExtents=<20 20 20>
      constraintSetup {angularLimitsMax=<10 35 20>, angularLimitsMin=<-10 -35 -60>, angularTarget=<1 0 0>}
      localJointOffset=<-5 0 0>
     }]
    relativeSpaceBone {boneName=pelvis}
    simSpaceSettings {simSpaceAngularAlpha=0.5}
   }
   n43 = Anim Dynamics - Bone: dyn_inSleeveFur_l!(ComponentPose=@n42, GravityOverride:Vector="0,0,-980") -> (Pose) {
    alpha=0.75
    angularDampingOverride=1
    angularSpringConstant=1200
    bAngularSpring=true
    bOverrideAngularDamping=true
    bUseGravityOverride=true
    gravityScale=0.35
    simulationSpace=RootRelative
    alphaBoolBlend {blendOption=Linear}
    alphaScaleBiasClamp {bInterpResult=true}
    boundBone {boneName=dyn_inSleeveFur_l}
    componentAppliedLinearAccClamp=<3000 3000 3000>
    componentLinearAccScale=<0.3 0.3 0.3>
    componentLinearVelScale=<0.3 0.3 0.3>
    gravityOverride {z=-980}
    physicsBodyDefinitions [{
      boundBone {boneName=dyn_inSleeveFur_l}
      boxExtents=<20 20 20>
      constraintSetup {angularLimitsMax=<10 35 20>, angularLimitsMin=<-10 -35 -60>, angularTarget=<1 0 0>}
      localJointOffset=<-5 0 0>
     }]
    relativeSpaceBone {boneName=pelvis}
    simSpaceSettings {simSpaceAngularAlpha=0.5}
   }
   n44 = Anim Dynamics - Bone: dyn_topSleeveFur_l!(ComponentPose=@n43, GravityOverride:Vector="0,0,-980") -> (Pose) {
    alpha=0.75
    angularDampingOverride=1
    angularSpringConstant=1200
    bAngularSpring=true
    bOverrideAngularDamping=true
    bUseGravityOverride=true
    gravityScale=0.35
    simulationSpace=RootRelative
    alphaBoolBlend {blendOption=Linear}
    alphaScaleBiasClamp {bInterpResult=true}
    boundBone {boneName=dyn_topSleeveFur_l}
    componentAppliedLinearAccClamp=<3000 3000 3000>
    componentLinearAccScale=<0.3 0.3 0.3>
    componentLinearVelScale=<0.3 0.3 0.3>
    gravityOverride {z=-980}
    physicsBodyDefinitions [{
      boundBone {boneName=dyn_topSleeveFur_l}
      boxExtents=<20 20 20>
      constraintSetup {angularLimitsMax=<10 35 20>, angularLimitsMin=<-10 -35 -60>, angularTarget=<1 0 0>}
      localJointOffset=<-5 0 0>
     }]
    relativeSpaceBone {boneName=pelvis}
    simSpaceSettings {simSpaceAngularAlpha=0.5}
   }
   n45 = Anim Dynamics - Bone: dyn_bottomSleeveFur_l!(ComponentPose=@n44, GravityOverride:Vector="0,0,-980") -> (Pose) {
    alpha=0.75
    angularDampingOverride=1
    angularSpringConstant=1200
    bAngularSpring=true
    bOverrideAngularDamping=true
    bUseGravityOverride=true
    gravityScale=0.35
    simulationSpace=RootRelative
    alphaBoolBlend {blendOption=Linear}
    alphaScaleBiasClamp {bInterpResult=true}
    boundBone {boneName=dyn_bottomSleeveFur_l}
    componentAppliedLinearAccClamp=<3000 3000 3000>
    componentLinearAccScale=<0.3 0.3 0.3>
    componentLinearVelScale=<0.3 0.3 0.3>
    gravityOverride {z=-980}
    physicsBodyDefinitions [{
      boundBone {boneName=dyn_bottomSleeveFur_l}
      boxExtents=<20 20 20>
      constraintSetup {angularLimitsMax=<10 35 20>, angularLimitsMin=<-10 -35 -60>, angularTarget=<1 0 0>}
      localJointOffset=<-5 0 0>
     }]
    relativeSpaceBone {boneName=pelvis}
    simSpaceSettings {simSpaceAngularAlpha=0.5}
   }
   n46 = Anim Dynamics - Bone: dyn_outSleeveFur_r!(ComponentPose=@n45, GravityOverride:Vector="0,0,-980") -> (Pose) {
    alpha=0.75
    angularDampingOverride=1
    angularSpringConstant=1200
    bAngularSpring=true
    bOverrideAngularDamping=true
    bUseGravityOverride=true
    gravityScale=0.35
    simulationSpace=RootRelative
    alphaBoolBlend {blendOption=Linear}
    alphaScaleBiasClamp {bInterpResult=true}
    boundBone {boneName=dyn_outSleeveFur_r}
    componentAppliedLinearAccClamp=<3000 3000 3000>
    componentLinearAccScale=<0.3 0.3 0.3>
    componentLinearVelScale=<0.3 0.3 0.3>
    gravityOverride {z=-980}
    physicsBodyDefinitions [{
      boundBone {boneName=dyn_outSleeveFur_r}
      boxExtents=<20 20 20>
      constraintSetup {angularLimitsMax=<10 35 20>, angularLimitsMin=<-10 -35 -60>, angularTarget=<1 0 0>}
      localJointOffset=<-5 0 0>
     }]
    relativeSpaceBone {boneName=pelvis}
    simSpaceSettings {simSpaceAngularAlpha=0.5}
   }
   n47 = Anim Dynamics - Bone: dyn_inSleeveFur_r!(ComponentPose=@n46, GravityOverride:Vector="0,0,-980") -> (Pose) {
    alpha=0.75
    angularDampingOverride=1
    angularSpringConstant=1200
    bAngularSpring=true
    bOverrideAngularDamping=true
    bUseGravityOverride=true
    gravityScale=0.35
    simulationSpace=RootRelative
    alphaBoolBlend {blendOption=Linear}
    alphaScaleBiasClamp {bInterpResult=true}
    boundBone {boneName=dyn_inSleeveFur_r}
    componentAppliedLinearAccClamp=<3000 3000 3000>
    componentLinearAccScale=<0.3 0.3 0.3>
    componentLinearVelScale=<0.3 0.3 0.3>
    gravityOverride {z=-980}
    physicsBodyDefinitions [{
      boundBone {boneName=dyn_inSleeveFur_r}
      boxExtents=<20 20 20>
      constraintSetup {angularLimitsMax=<10 35 20>, angularLimitsMin=<-10 -35 -60>, angularTarget=<1 0 0>}
      localJointOffset=<-5 0 0>
     }]
    relativeSpaceBone {boneName=pelvis}
    simSpaceSettings {simSpaceAngularAlpha=0.5}
   }
   n48 = Anim Dynamics - Bone: dyn_topSleeveFur_r!(ComponentPose=@n47, GravityOverride:Vector="0,0,-980") -> (Pose) {
    alpha=0.75
    angularDampingOverride=1
    angularSpringConstant=1200
    bAngularSpring=true
    bOverrideAngularDamping=true
    bUseGravityOverride=true
    gravityScale=0.35
    simulationSpace=RootRelative
    alphaBoolBlend {blendOption=Linear}
    alphaScaleBiasClamp {bInterpResult=true}
    boundBone {boneName=dyn_topSleeveFur_r}
    componentAppliedLinearAccClamp=<3000 3000 3000>
    componentLinearAccScale=<0.3 0.3 0.3>
    componentLinearVelScale=<0.3 0.3 0.3>
    gravityOverride {z=-980}
    physicsBodyDefinitions [{
      boundBone {boneName=dyn_topSleeveFur_r}
      boxExtents=<20 20 20>
      constraintSetup {angularLimitsMax=<10 35 20>, angularLimitsMin=<-10 -35 -60>, angularTarget=<1 0 0>}
      localJointOffset=<-5 0 0>
     }]
    relativeSpaceBone {boneName=pelvis}
    simSpaceSettings {simSpaceAngularAlpha=0.5}
   }
   n49 = Anim Dynamics - Bone: dyn_bottomSleeveFur_r!(ComponentPose=@n48, GravityOverride:Vector="0,0,-980") -> (Pose) {
    alpha=0.75
    angularDampingOverride=1
    angularSpringConstant=1200
    bAngularSpring=true
    bOverrideAngularDamping=true
    bUseGravityOverride=true
    gravityScale=0.35
    simulationSpace=RootRelative
    alphaBoolBlend {blendOption=Linear}
    alphaScaleBiasClamp {bInterpResult=true}
    boundBone {boneName=dyn_bottomSleeveFur_r}
    componentAppliedLinearAccClamp=<3000 3000 3000>
    componentLinearAccScale=<0.3 0.3 0.3>
    componentLinearVelScale=<0.3 0.3 0.3>
    gravityOverride {z=-980}
    physicsBodyDefinitions [{
      boundBone {boneName=dyn_bottomSleeveFur_r}
      boxExtents=<20 20 20>
      constraintSetup {angularLimitsMax=<10 35 20>, angularLimitsMin=<-10 -35 -60>, angularTarget=<1 0 0>}
      localJointOffset=<-5 0 0>
     }]
    relativeSpaceBone {boneName=pelvis}
    simSpaceSettings {simSpaceAngularAlpha=0.5}
   }
   n50 = Get isFallingNotMoving()
   n51 = Anim Dynamics - Bone: dyn_belt01!(ComponentPose=@n49, Alpha=@n50, GravityOverride:Vector="0,0,-980") -> (Pose) {
    angularDampingOverride=1
    angularSpringConstant=20000
    bOverrideAngularDamping=true
    bUseGravityOverride=true
    gravityScale=0.35
    simulationSpace=RootRelative
    alphaBoolBlend {blendInTime=0.1, blendOption=Linear, blendOutTime=0.2}
    alphaScaleBiasClamp {bInterpResult=true}
    boundBone {boneName=dyn_belt01}
    componentAppliedLinearAccClamp=<4000 4000 3500>
    componentLinearAccScale=<1 1 0.6>
    componentLinearVelScale=<1 1 0.6>
    gravityOverride {z=-980}
    physicsBodyDefinitions [{
      boundBone {boneName=dyn_belt01}
      boxExtents=<28 28 28>
      constraintSetup {angularLimitsMax=<0 7.5 1>, angularLimitsMin=<0 0 0>, angularTarget=<1 0 0>}
      localJointOffset=<2 0 0>
     }]
    relativeSpaceBone {boneName=pelvis}
    simSpaceSettings {simSpaceAngularAlpha=1}
   }
   n52 = Anim Dynamics - Bone: dyn_belt02!(ComponentPose=@n51, Alpha=@n50, GravityOverride:Vector="0,0,-980") -> (Pose) {
    angularDampingOverride=1
    angularSpringConstant=20000
    bOverrideAngularDamping=true
    bUseGravityOverride=true
    gravityScale=0.35
    simulationSpace=RootRelative
    alphaBoolBlend {blendInTime=0.1, blendOption=Linear, blendOutTime=0.2}
    alphaScaleBiasClamp {bInterpResult=true}
    boundBone {boneName=dyn_belt02}
    componentAppliedLinearAccClamp=<4000 4000 3500>
    componentLinearAccScale=<1 1 0.6>
    componentLinearVelScale=<1 1 0.6>
    gravityOverride {z=-980}
    physicsBodyDefinitions [{
      boundBone {boneName=dyn_belt02}
      boxExtents=<28 28 28>
      constraintSetup {angularLimitsMax=<0 6.5 1>, angularLimitsMin=<0 0 0>, angularTarget=<1 0 0>}
      localJointOffset=<2 0 0>
     }]
    relativeSpaceBone {boneName=pelvis}
    simSpaceSettings {simSpaceAngularAlpha=1}
   }
   n53 = Anim Dynamics - Bone: dyn_belt03!(ComponentPose=@n52, Alpha=@n50, GravityOverride:Vector="0,0,-980") -> (Pose) {
    angularDampingOverride=1
    angularSpringConstant=20000
    bOverrideAngularDamping=true
    bUseGravityOverride=true
    gravityScale=0.35
    simulationSpace=RootRelative
    alphaBoolBlend {blendInTime=0.1, blendOption=Linear, blendOutTime=0.2}
    alphaScaleBiasClamp {bInterpResult=true}
    boundBone {boneName=dyn_belt03}
    componentAppliedLinearAccClamp=<4000 4000 3500>
    componentLinearAccScale=<1 1 0.6>
    componentLinearVelScale=<1 1 0.6>
    gravityOverride {z=-980}
    physicsBodyDefinitions [{
      boundBone {boneName=dyn_belt03}
      boxExtents=<28 28 28>
      constraintSetup {angularLimitsMax=<0 5 0>, angularLimitsMin=<0 0 0>, angularTarget=<1 0 0>}
      localJointOffset=<2 0 0>
     }]
    relativeSpaceBone {boneName=pelvis}
    simSpaceSettings {simSpaceAngularAlpha=1}
   }
   n54 = "Transform (Modify) Bone - Bone: base_holster"!(ComponentPose=@n53, bAlphaBoolEnabled:bool=True, Translation="-1.936,0.000,-0.300", Rotation="0,0,0", Scale="1,1,1") -> (Pose) {
    alphaInputType=Bool
    rotationMode=BMM_Additive
    rotationSpace=BCS_BoneSpace
    translationMode=BMM_Additive
    translationSpace=BCS_BoneSpace
    boneToModify {boneName=base_holster}
    translation {x=-1.936, z=-0.3}
   }
   n55 = "Transform (Modify) Bone - Bone: base_belt_l"!(ComponentPose=@n54, bAlphaBoolEnabled:bool=True, Translation="-1.872,-2.148,0.000", Rotation="0,0,0", Scale="1,1,1") -> (Pose) {
    alphaInputType=Bool
    rotationMode=BMM_Additive
    rotationSpace=BCS_BoneSpace
    translationMode=BMM_Additive
    translationSpace=BCS_BoneSpace
    alphaBoolBlend {blendInTime=5, blendOutTime=0.6}
    boneToModify {boneName=base_belt_l}
    translation=<-1.872 -2.148>
   }
   n56 = Anim Dynamics - Bone: dyn_holster(ComponentPose=@n55, Alpha=1, ExternalForce="0,0,0") {
    angularDampingOverride=1
    angularSpringConstant=17000
    bAngularSpring=true
    bOverrideAngularDamping=true
    bUseGravityOverride=true
    gravityScale=0.35
    simulationSpace=RootRelative
    boundBone {boneName=dyn_holster}
    componentAppliedLinearAccClamp=<3000 3000 3000>
    componentLinearAccScale=<0.3 0.3 0.3>
    componentLinearVelScale=<0.3 0.3 0.3>
    gravityOverride {z=-980}
    physicsBodyDefinitions [{
      boundBone {boneName=dyn_holster}
      boxExtents=<35 35 35>
      constraintSetup {angularLimitsMax=<0 0 6>, angularLimitsMin=<0 -5 -6>, angularTarget=<1 0 0>}
      localJointOffset=<2 0 0>
     }]
    simSpaceSettings {simSpaceAngularAlpha=0.15}
   }
   n57 = Anim Dynamics - Bone: dyn_heelBullet_l!(ComponentPose=@n56, Alpha=1, GravityOverride:Vector="0,0,-980", ExternalForce="0,0,0") -> (Pose) {
    angularDampingOverride=0.999
    angularSpringConstant=6000
    bAngularSpring=true
    bOverrideAngularDamping=true
    bUseGravityOverride=true
    gravityScale=0.35
    simulationSpace=BoneRelative
    boundBone {boneName=dyn_heelBullet_l}
    componentAppliedLinearAccClamp=<3000 3000 3000>
    componentLinearAccScale=<0.5 0.5 0.5>
    componentLinearVelScale=<0.5 0.5 0.5>
    gravityOverride {z=-980}
    physicsBodyDefinitions [{
      boundBone {boneName=dyn_heelBullet_l}
      boxExtents=<15 15 15>
      constraintSetup {angularLimitsMax=<0 40 0>, angularLimitsMin=<0 -40 -60>, angularTarget=<1 0 0>}
      localJointOffset=<3 -0.5 0.5>
     }]
    relativeSpaceBone {boneName=heelFoot_l}
    simSpaceSettings {simSpaceAngularAlpha=0.2}
   }
   n58 = Anim Dynamics - Bone: dyn_heelBullet_r!(ComponentPose=@n57, Alpha=1, GravityOverride:Vector="0,0,-980", ExternalForce="0,0,0") -> (Pose) {
    angularDampingOverride=0.999
    angularSpringConstant=6000
    bAngularSpring=true
    bOverrideAngularDamping=true
    bUseGravityOverride=true
    gravityScale=0.35
    simulationSpace=BoneRelative
    boundBone {boneName=dyn_heelBullet_r}
    componentAppliedLinearAccClamp=<3000 3000 3000>
    componentLinearAccScale=<0.5 0.5 0.5>
    componentLinearVelScale=<0.5 0.5 0.5>
    gravityOverride {z=-980}
    physicsBodyDefinitions [{
      boundBone {boneName=dyn_heelBullet_r}
      boxExtents=<15 15 15>
      constraintSetup {angularLimitsMax=<0 40 0>, angularLimitsMin=<0 -40 -60>, angularTarget=<1 0 0>}
      localJointOffset=<3 -0.5 0.5>
     }]
    relativeSpaceBone {boneName=heelFoot_r}
    simSpaceSettings {simSpaceAngularAlpha=0.2}
   }
   n59 = Anim Dynamics - Bone: dyn_bracelet(ComponentPose=@n58, Alpha=1, ExternalForce="0,0,0") {
    angularDampingOverride=1
    angularSpringConstant=15000
    bAngularSpring=true
    bOverrideAngularDamping=true
    bUseGravityOverride=true
    gravityScale=0.35
    simulationSpace=RootRelative
    boundBone {boneName=dyn_bracelet}
    componentAppliedLinearAccClamp=<6000 6000 6000>
    componentLinearAccScale=<0.7 0.7 0.9>
    componentLinearVelScale=<0.7 0.7 0.9>
    gravityOverride {z=-980}
    physicsBodyDefinitions [{
      boundBone {boneName=dyn_bracelet}
      boxExtents=<29 29 29>
      constraintSetup {angularLimitsMax=<0 5 13>, angularLimitsMin=<0 -12 -13>, angularTarget=<1 0 0>}
      localJointOffset=<4 0 0>
     }]
    simSpaceSettings {simSpaceAngularAlpha=0.44}
   }
   n60 = "Transform (Modify) Bone - Bone: dyn_bracelet"!(ComponentPose=@n59, bAlphaBoolEnabled:bool=True, Translation="0,0,0", Rotation="15,0,0", Scale="1,1,1") -> (Pose) {
    alphaInputType=Bool
    rotationMode=BMM_Additive
    rotationSpace=BCS_BoneSpace
    translationMode=BMM_Additive
    translationSpace=BCS_BoneSpace
    alphaBoolBlend {blendInTime=0.1, blendOutTime=0.1}
    boneToModify {boneName=dyn_bracelet}
    rotation {pitch=15}
   }
   n61 = RigidBody!(ComponentPose=@n60, Alpha, OverrideWorldGravity:Vector="0,0,0", ExternalForce) -> (Pose) {
    bOverrideWorldGravity=true
    overridePhysicsAsset=&/F_MED_LivelySugar_Shore_Physics
    simulationSpace=BaseBoneSpace
    baseBoneRef {boneName=pelvis}
    componentAppliedLinearAccClamp=<1000 1000 1000>
    simSpaceSettings {
     maxAngularAcceleration=45
     maxAngularVelocity=8
     maxLinearAcceleration=1000
     maxLinearVelocity=300
     worldAlpha=1
     externalLinearDragV=<1 1 0.8>
    }
   }
   n62 = "Transform (Modify) Bone - Bone: base_handkerchiefBottom"!(ComponentPose=@n61, bAlphaBoolEnabled:bool=True, Translation="0,0,0", Rotation="-0.000,-0.022,5.273", Scale="1,1,1") -> (Pose) {
    alphaInputType=Bool
    rotationMode=BMM_Additive
    rotationSpace=BCS_BoneSpace
    translationMode=BMM_Additive
    translationSpace=BCS_BoneSpace
    boneToModify {boneName=base_handkerchiefBottom}
    rotation=<rot -0.0001 -0.022 5.273>
   }
   n63 = "Transform (Modify) Bone - Bone: base_handkerchiefTop"!(ComponentPose=@n62, bAlphaBoolEnabled:bool=True, Translation="0,0,0", Rotation="4.908,-9.467,11.603", Scale="1,1,1") -> (Pose) {
    alphaInputType=Bool
    rotationMode=BMM_Additive
    rotationSpace=BCS_BoneSpace
    translationMode=BMM_Additive
    translationSpace=BCS_BoneSpace
    boneToModify {boneName=base_handkerchiefTop}
    rotation=<rot 4.908 -9.4669 11.603>
   }
   n64 = Anim Dynamics - Bone: dyn_necklace(ComponentPose=@n63, Alpha=1, ExternalForce="0,0,0") {
    angularDampingOverride=1
    angularSpringConstant=18000
    bAngularSpring=true
    bOverrideAngularDamping=true
    bUseGravityOverride=true
    gravityScale=0.35
    simulationSpace=RootRelative
    boundBone {boneName=dyn_necklace}
    componentAppliedLinearAccClamp=<5000 5000 5000>
    componentLinearAccScale=<0.8 0.8 0.8>
    componentLinearVelScale=<0.8 0.8 0.8>
    gravityOverride {z=-980}
    physicsBodyDefinitions [{
      boundBone {boneName=dyn_necklace}
      boxExtents=<20 26 26>
      constraintSetup {angularLimitsMax=<15 0 0>, angularLimitsMin=<-15 0 -18>, angularTarget=<1 0 0>}
      localJointOffset=<4.3 0 0>
     }]
    simSpaceSettings {simSpaceAngularAlpha=0.5}
   }
   n65 = "Transform (Modify) Bone - Bone: dyn_necklace"!(ComponentPose=@n64, bAlphaBoolEnabled:bool=True, Translation="0,0,0", Rotation="0,-2,0", Scale="1,1,1") -> (Pose) {
    alphaInputType=Bool
    rotationMode=BMM_Additive
    rotationSpace=BCS_BoneSpace
    translationMode=BMM_Additive
    translationSpace=BCS_BoneSpace
    alphaBoolBlend {blendInTime=0.1, blendOutTime=0.1}
    boneToModify {boneName=dyn_necklace}
    rotation {yaw=-2}
   }
   Save cached pose 'allDynCache'(Pose=Component To Local(ComponentPose=@n65)) {cacheName=allDynCache}
  ]}
 EventGraph {nodes [
   Event BlueprintUpdateAnimation()
   > n1 = "Gravity Override State Function (Parameter Asset) (Nativized)"(
    CharacterPartAnimInstance=Self-Reference()
    GameTimeInSeconds=GetGameTimeInSeconds()
    DebugJointName=spine_05
    Parameters=&/F_MED_LivelySugar_Shore_Parameters
   )
   > Set Gravity Override(Gravity Override=@n1)
   n2 = Get bIsSkyDiving()
   n3 = Get bIsFalling()
   n4 = Get bIsJumping()
   > n5 = Branch(Condition=AND Boolean!(A=OR Boolean(A=@n3, B=@n4), B=NOT Boolean(A=@n2), C:bool=true) -> (ReturnValue))
   > n6 = Branch(Condition=Get bIsTacticalSprinting())
   > Set isFallingNotMoving(isFallingNotMoving:double=0.05) -> (Output_Get:double)
   Set isFallingNotMoving(execute=@n6.else, isFallingNotMoving=0.7)
   Set isFallingNotMoving(execute=@n5.else, isFallingNotMoving=0.3)
   NOT Boolean(A=Get bIsMoving2D())
  ]}
]
variables [
 Gravity Override {type=Vector}
 isFallingNotMoving {type=number}
 Sim Space Settings {
  type=AnimPhysSimSpaceSettings
  default {maxAngularAcceleration=45, maxAngularVelocity=8, simSpaceAngularAlpha=0.3, externalAngularVelocity=<0 0 0>}
 }
]
